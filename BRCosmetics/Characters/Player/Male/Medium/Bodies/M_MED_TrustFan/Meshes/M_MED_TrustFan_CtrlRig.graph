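# Control Rig for the M_MED_TrustFan player body cosmetic that procedurally poses its robe and sleeve bones from arm and leg motion.
# Spherical pose readers, calf bend and foot/pelvis positions relative to root fill driver variables, with Kalman smoothing on foot-crossing values.
# Those drivers weight parent constraints and transform offsets, blending center robe chains between left and right and pulling panels toward the opposite leg when feet cross.
# Sleeve bones follow the lower arms, and skydiving scales some rear-robe corrections.
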ControlRigBlueprint /BRCosmetics/Characters/Player/Male/Medium/Bodies/M_MED_TrustFan/Meshes/M_MED_TrustFan_CtrlRig
graphs [
 Multiply_4_ContainedGraph {nodes [
   n1 = RigVMFunctionEntryNode()
   RigVMFunctionReturnNode(Result=&RigVMFunction_MathDoubleMul(B=@n1.C, A=&RigVMFunction_MathDoubleMul(A=@n1.A, B=@n1.B).Result).Result)
  ]}
 Multiply_4_ContainedGraph {nodes [
   n1 = RigVMFunctionEntryNode()
   RigVMFunctionReturnNode(Result=&RigVMFunction_MathDoubleMul(B=@n1.C, A=&RigVMFunction_MathDoubleMul(A=@n1.A, B=@n1.B).Result).Result)
  ]}
 New Function_ContainedGraph {nodes [
   RigVMFunctionEntryNode()
   > RigVMFunction_Sequence = RigVMAggregateNode()
   &RigUnit_ParentConstraint(
    AdvancedSettings=(InterpolationType=Average,RotationOrderForFilter=XZY)
    Child=(Type=Bone,Name="dyn_sleeve_l_ft_03")
    Parents=((Item=(Type=Bone,Name="lowerarm_l"),Weight=1))
    Weight=1
    ExecutePin=@RigVMFunction_Sequence.A
   )
   > &RigUnit_ParentConstraint(
    AdvancedSettings=(InterpolationType=Average,RotationOrderForFilter=XZY)
    Child=(Type=Bone,Name="dyn_sleeve_l_bt_03")
    Parents=((Item=(Type=Bone,Name="lowerarm_l"),Weight=1))
    Weight=1
   )
   > &RigUnit_ParentConstraint(
    AdvancedSettings=(InterpolationType=Average,RotationOrderForFilter=XZY)
    Child=(Type=Bone,Name="dyn_sleeve_l_tp_03")
    Parents=((Item=(Type=Bone,Name="lowerarm_l"),Weight=1))
    Weight=1
   )
   > &RigUnit_ParentConstraint(
    AdvancedSettings=(InterpolationType=Average,RotationOrderForFilter=XZY)
    Child=(Type=Bone,Name="dyn_sleeve_l_bk_03")
    Parents=((Item=(Type=Bone,Name="lowerarm_l"),Weight=1))
    Weight=1
   )
   &RigUnit_ParentConstraint(
    AdvancedSettings=(InterpolationType=Average,RotationOrderForFilter=XZY)
    Child=(Type=Bone,Name="dyn_sleeve_r_ft_03")
    Parents=((Item=(Type=Bone,Name="lowerarm_r"),Weight=1))
    Weight=1
    ExecutePin=@RigVMFunction_Sequence.B
   )
   > &RigUnit_ParentConstraint(
    AdvancedSettings=(InterpolationType=Average,RotationOrderForFilter=XZY)
    Child=(Type=Bone,Name="dyn_sleeve_r_bt_03")
    Parents=((Item=(Type=Bone,Name="lowerarm_r"),Weight=1))
    Weight=1
   )
   > &RigUnit_ParentConstraint(
    AdvancedSettings=(InterpolationType=Average,RotationOrderForFilter=XZY)
    Child=(Type=Bone,Name="dyn_sleeve_r_tp_03")
    Parents=((Item=(Type=Bone,Name="lowerarm_r"),Weight=1))
    Weight=1
   )
   > &RigUnit_ParentConstraint(
    AdvancedSettings=(InterpolationType=Average,RotationOrderForFilter=XZY)
    Child=(Type=Bone,Name="dyn_sleeve_r_bk_03")
    Parents=((Item=(Type=Bone,Name="lowerarm_r"),Weight=1))
    Weight=1
   )
   RigVMFunctionReturnNode(ExecuteContext=@RigVMFunction_Sequence.C)
  ]}
 New Function_ContainedGraph {nodes [
   RigVMFunctionEntryNode()
   > RigVMFunction_Sequence = RigVMAggregateNode()
   n1 = &RigUnit_SphericalPoseReader(
    ActiveRegionScaleFactors=(PositiveWidth=4,NegativeWidth=1,PositiveHeight=2.5,NegativeHeight=2)
    ActiveRegionSize=0.1
    Debug=(bDrawDebug=false,bDraw2D=False,bDrawLocalAxes=false,DebugScale=25,DebugSegments=20,DebugThickness=0.25)
    DriverAxis=(X=-1,Y=0,Z=0)
    DriverItem=(Type=Bone,Name="thigh_l")
    FalloffRegionScaleFactors=(PositiveWidth=1,NegativeWidth=2,PositiveHeight=1,NegativeHeight=0.5)
    FalloffSize=0.4
    FlipHeightScaling=False
    FlipWidthScaling=False
    OptionalParentItem=(Type=Bone,Name="None")
    RotationOffset=(X=100,Y=0,Z=-20)
    ExecutePin=@RigVMFunction_Sequence.A
   )
   > RigVMVariableNode(Value=@n1.OutputParam) {variable=Thigh_l_forward_Standing}
   > n2 = &RigUnit_SphericalPoseReader(
    ActiveRegionScaleFactors=(PositiveWidth=2,NegativeWidth=1,PositiveHeight=2,NegativeHeight=2)
    ActiveRegionSize=0.1
    Debug=(bDrawDebug=false,bDraw2D=False,bDrawLocalAxes=False,DebugScale=25,DebugSegments=20,DebugThickness=0.25)
    DriverAxis=(X=-1,Y=0,Z=0)
    DriverItem=(Type=Bone,Name="thigh_l")
    FalloffRegionScaleFactors=(PositiveWidth=1,NegativeWidth=1,PositiveHeight=1,NegativeHeight=1)
    FalloffSize=0.3
    FlipHeightScaling=False
    FlipWidthScaling=False
    OptionalParentItem=(Type=Bone,Name="None")
    RotationOffset=(X=105,Y=0,Z=0)
   )
   > RigVMVariableNode(Value=@n2.OutputParam) {variable=Thigh_l_forward}
   > n3 = &RigUnit_SphericalPoseReader(
    ActiveRegionScaleFactors=(PositiveWidth=4,NegativeWidth=1,PositiveHeight=2.5,NegativeHeight=2)
    ActiveRegionSize=0.1
    Debug=(bDrawDebug=false,bDraw2D=False,bDrawLocalAxes=False,DebugScale=25,DebugSegments=20,DebugThickness=0.25)
    DriverAxis=(X=1,Y=0,Z=0)
    DriverItem=(Type=Bone,Name="thigh_r")
    FalloffRegionScaleFactors=(PositiveWidth=1,NegativeWidth=2,PositiveHeight=1,NegativeHeight=0.5)
    FalloffSize=0.4
    FlipHeightScaling=False
    FlipWidthScaling=true
    OptionalParentItem=(Type=Bone,Name="None")
    RotationOffset=(X=-80,Y=0,Z=-20)
   )
   > RigVMVariableNode(Value=@n3.OutputParam) {variable=Thigh_r_forward_Standing}
   > n4 = &RigUnit_SphericalPoseReader(
    ActiveRegionScaleFactors=(PositiveWidth=1,NegativeWidth=2,PositiveHeight=2,NegativeHeight=2)
    ActiveRegionSize=0.1
    Debug=(bDrawDebug=false,bDraw2D=False,bDrawLocalAxes=False,DebugScale=25,DebugSegments=20,DebugThickness=0.25)
    DriverAxis=(X=1,Y=0,Z=0)
    DriverItem=(Type=Bone,Name="thigh_r")
    FalloffRegionScaleFactors=(PositiveWidth=0.75,NegativeWidth=1,PositiveHeight=1,NegativeHeight=1)
    FalloffSize=0.3
    FlipHeightScaling=False
    FlipWidthScaling=False
    OptionalParentItem=(Type=Bone,Name="None")
    RotationOffset=(X=-75,Y=0,Z=0)
   )
   > RigVMVariableNode(Value=@n4.OutputParam) {variable=Thigh_r_forward}
   > n5 = &RigUnit_SphericalPoseReader(
    ActiveRegionScaleFactors=(PositiveWidth=2,NegativeWidth=1.5,PositiveHeight=2,NegativeHeight=1)
    ActiveRegionSize=0.125
    Debug=(bDrawDebug=false,bDraw2D=false,DebugScale=25,DebugSegments=20,DebugThickness=0.25)
    DriverAxis=(X=-1,Y=0,Z=0)
    DriverItem=(Type=Bone,Name="thigh_l")
    FalloffRegionScaleFactors=(PositiveWidth=1,NegativeWidth=1.25,PositiveHeight=1,NegativeHeight=0.5)
    FalloffSize=0.35
    OptionalParentItem=(Type=Bone,Name="pelvis")
    RotationOffset=(X=150,Y=180,Z=10)
   )
   > RigVMVariableNode(Value=@n5.OutputParam) {variable=Thigh_l_Inner}
   > n6 = &RigUnit_SphericalPoseReader(
    ActiveRegionScaleFactors=(PositiveWidth=2,NegativeWidth=1.5,PositiveHeight=1,NegativeHeight=2)
    ActiveRegionSize=0.125
    Debug=(bDrawDebug=false,bDraw2D=false,DebugScale=25,DebugSegments=20,DebugThickness=0.25)
    DriverAxis=(X=1,Y=0,Z=0)
    DriverItem=(Type=Bone,Name="thigh_r")
    FalloffRegionScaleFactors=(PositiveWidth=1.5,NegativeWidth=1.25,PositiveHeight=0.5,NegativeHeight=1)
    FalloffSize=0.35
    OptionalParentItem=(Type=Bone,Name="pelvis")
    RotationOffset=(X=-150,Y=0,Z=10)
   )
   > RigVMVariableNode(Value=@n6.OutputParam) {variable=Thigh_r_Inner}
   > n7 = &RigUnit_SphericalPoseReader(
    ActiveRegionScaleFactors=(PositiveWidth=3,NegativeWidth=2,PositiveHeight=2,NegativeHeight=2)
    ActiveRegionSize=0.1
    Debug=(bDrawDebug=false,bDraw2D=False,bDrawLocalAxes=False,DebugScale=25,DebugSegments=20,DebugThickness=0.25)
    DriverAxis=(X=-1,Y=0,Z=0)
    DriverItem=(Type=Bone,Name="thigh_l")
    FalloffRegionScaleFactors=(PositiveWidth=2,NegativeWidth=2,PositiveHeight=2,NegativeHeight=2)
    FalloffSize=0.25
    FlipHeightScaling=false
    FlipWidthScaling=false
    OptionalParentItem=(Type=Bone,Name="None")
    RotationOffset=(X=180,Y=-40,Z=0)
   )
   > RigVMVariableNode(Value=@n7.OutputParam) {variable=Thigh_l_Side}
   > n8 = &RigUnit_SphericalPoseReader(
    ActiveRegionScaleFactors=(PositiveWidth=3,NegativeWidth=2,PositiveHeight=2,NegativeHeight=2)
    ActiveRegionSize=0.1
    Debug=(bDrawDebug=false,bDraw2D=False,bDrawLocalAxes=False,DebugScale=25,DebugSegments=20,DebugThickness=0.25)
    DriverAxis=(X=1,Y=0,Z=0)
    DriverItem=(Type=Bone,Name="thigh_r")
    FalloffRegionScaleFactors=(PositiveWidth=2,NegativeWidth=2,PositiveHeight=2,NegativeHeight=2)
    FalloffSize=0.25
    FlipHeightScaling=False
    FlipWidthScaling=true
    OptionalParentItem=(Type=Bone,Name="None")
    RotationOffset=(X=0,Y=-40,Z=0)
   )
   > RigVMVariableNode(Value=@n8.OutputParam) {variable=Thigh_r_Side}
   RigVMFunction_MathQuaternionToEuler = &RigVMFunction_MathQuaternionToEuler(
    RotationOrder=ZYX
    Value=&RigUnit_GetRelativeTransformForItem(Child=(Type=Bone,Name="calf_l"), Parent=(Type=Bone,Name="thigh_l"), bChildInitial=False, bParentInitial=False).RelativeTransform.Rotation
   )
   Remap = &RigVMFunction_MathDoubleRemap(SourceMaximum=110, SourceMinimum=25, TargetMaximum=1, TargetMinimum=0, Value=@RigVMFunction_MathQuaternionToEuler.Result.Z)
   > RigVMVariableNode(Value=@Remap.Result) {variable=Calf_l_Bent}
   n9 = &RigUnit_GetRelativeTransformForItem(Child=(Type=Bone,Name="calf_r"), Parent=(Type=Bone,Name="thigh_r"), bChildInitial=False, bParentInitial=False)
   > RigVMVariableNode(Value=&RigVMFunction_MathDoubleRemap(
     SourceMaximum=110
     SourceMinimum=25
     TargetMaximum=1
     TargetMinimum=0
     Value=&RigVMFunction_MathQuaternionToEuler(RotationOrder=ZYX, Value=@n9.RelativeTransform.Rotation).Result.Z
    ).Result) {variable=Calf_r_Bent}
   > n10 = &RigUnit_SphericalPoseReader(
    ActiveRegionScaleFactors=(PositiveWidth=2,NegativeWidth=2,PositiveHeight=1.5,NegativeHeight=1)
    ActiveRegionSize=0.15
    Debug=(bDrawDebug=false,bDraw2D=False,bDrawLocalAxes=False,DebugScale=25,DebugSegments=20,DebugThickness=0.25)
    DriverAxis=(X=-1,Y=0,Z=0)
    DriverItem=(Type=Bone,Name="thigh_l")
    FalloffRegionScaleFactors=(PositiveWidth=2,NegativeWidth=2,PositiveHeight=4,NegativeHeight=1)
    FalloffSize=0.4
    FlipHeightScaling=false
    FlipWidthScaling=false
    OptionalParentItem=(Type=Bone,Name="None")
    RotationOffset=(X=0,Y=90,Z=80)
   )
   > RigVMVariableNode(Value=@n10.OutputParam) {variable=OutputParam}
   > n11 = &RigUnit_SphericalPoseReader(
    ActiveRegionScaleFactors=(PositiveWidth=2,NegativeWidth=2,PositiveHeight=1.5,NegativeHeight=1)
    ActiveRegionSize=0.15
    Debug=(bDrawDebug=false,bDraw2D=False,bDrawLocalAxes=False,DebugScale=25,DebugSegments=20,DebugThickness=0.25)
    DriverAxis=(X=1,Y=0,Z=0)
    DriverItem=(Type=Bone,Name="thigh_r")
    FalloffRegionScaleFactors=(PositiveWidth=2,NegativeWidth=2,PositiveHeight=4,NegativeHeight=1)
    FalloffSize=0.4
    FlipHeightScaling=false
    FlipWidthScaling=false
    OptionalParentItem=(Type=Bone,Name="None")
    RotationOffset=(X=180,Y=90,Z=80)
   )
   > RigVMVariableNode(Value=@n11.OutputParam) {variable=OutputParam_0}
   n12 = &RigUnit_GetRelativeTransformForItem(Child=(Type=Bone,Name="foot_l"), Parent=(Type=Bone,Name="root"), bChildInitial=False, bParentInitial=False)
   > RigVMVariableNode(Value=&RigVMFunction_MathDoubleRemap(
     SourceMaximum=70
     SourceMinimum=25
     TargetMaximum=1
     TargetMinimum=0
     Value=&RigVMFunction_MathVectorDistance(B=(X=0,Y=20,Z=0), A=@n12.RelativeTransform.Translation).Result
    ).Result) {variable=Thigh_l_backward}
   RigVMFunction_MathVectorDistance = &RigVMFunction_MathVectorDistance(
    B=(X=0,Y=20,Z=0)
    A=&RigUnit_GetRelativeTransformForItem(Child=(Type=Bone,Name="foot_r"), Parent=(Type=Bone,Name="root"), bChildInitial=False, bParentInitial=false).RelativeTransform.Translation
   )
   > RigVMVariableNode(Value=&RigVMFunction_MathDoubleRemap(SourceMaximum=70, SourceMinimum=25, TargetMaximum=1, TargetMinimum=0, Value=@RigVMFunction_MathVectorDistance.Result).Result) {variable=Thigh_r_backward}
   n13 = &RigUnit_GetRelativeTransformForItem(Child=(Type=Bone,Name="spine_01"), Parent=(Type=Bone,Name="root"), bChildInitial=False, bParentInitial=false)
   > RigVMVariableNode(Value=&RigVMFunction_MathDoubleSub(
     A=1
     B=&RigVMFunction_MathDoubleRemap(SourceMaximum=30, SourceMinimum=60, TargetMaximum=1, TargetMinimum=0, Value=@n13.RelativeTransform.Translation.Z).Result
    ).Result) {variable=pelvisLowToGround_INVERSE}
   > RigVMVariableNode(Value=RigVMDispatchNode(False=1, True=0.25, Condition=RigVMVariableNode() {variable=IsSkydiving}.Value) {template="DISPATCH_RigVMDispatch_If(in Condition,in True,in False,out Result)"}.Result) {variable=Is Skydiving Alpha}
   n14 = &RigUnit_GetRelativeTransformForItem(Child=(Type=Bone,Name="foot_l"), Parent=(Type=Bone,Name="root"), bChildInitial=False, bParentInitial=False)
   Greater = /Script/RigVM.RigVMFunction_MathDoubleGreater(B=-5, A=@n14.RelativeTransform.Translation.Y) {is=struct RigVMUnitNode, template="Greater::Execute(in A,in B,out Result)"}
   Less = /Script/RigVM.RigVMFunction_MathDoubleLess(B=0, A=@n14.RelativeTransform.Translation.X) {is=struct RigVMUnitNode, template="Less::Execute(in A,in B,out Result)"}
   RigVMFunction_MathBoolAnd = &RigVMFunction_MathBoolAnd(B=@Greater.Result, A=@Less.Result)
   n15 = RigVMDispatchNode(False=0, True=@n14.RelativeTransform.Translation.X, Condition=@RigVMFunction_MathBoolAnd.Result) {template="DISPATCH_RigVMDispatch_If(in Condition,in True,in False,out Result)"}
   > RigVMVariableNode(Value=&RigVMFunction_KalmanFloat(
     BufferSize=2
     Value=&RigVMFunction_MathDoubleRemap(SourceMaximum=0, SourceMinimum=-35, TargetMaximum=0, TargetMinimum=1, Value=@n15.Result).Result
    ) {template="KalmanFilter::Execute(in Value,in BufferSize,out Result)"}.Result) {variable=L Foot Across Front}
   n16 = &RigVMFunction_MathDoubleLess(B=-2.5, A=@n14.RelativeTransform.Translation.X) {template="Less::Execute(in A,in B,out Result)"}
   n17 = &RigVMFunction_MathDoubleLess(B=-22, A=@n14.RelativeTransform.Translation.Y) {template="Less::Execute(in A,in B,out Result)"}
   n18 = &RigVMFunction_MathDoubleRemap(
    SourceMaximum=-10
    SourceMinimum=-40
    TargetMaximum=0
    TargetMinimum=1
    Value=RigVMDispatchNode(False=0, True=@n14.RelativeTransform.Translation.X, Condition=&RigVMFunction_MathBoolAnd(A=@n16.Result, B=@n17.Result).Result) {template="DISPATCH_RigVMDispatch_If(in Condition,in True,in False,out Result)"}.Result
   )
   n19 = &RigVMFunction_MathDoubleLess(B=-5, A=@n14.RelativeTransform.Translation.Y) {template="Less::Execute(in A,in B,out Result)"}
   n20 = &RigVMFunction_MathDoubleLess(B=-2.5, A=@n14.RelativeTransform.Translation.X) {template="Less::Execute(in A,in B,out Result)"}
   n21 = &RigVMFunction_MathDoubleRemap(
    SourceMaximum=-10
    SourceMinimum=-40
    TargetMaximum=0
    TargetMinimum=1
    Value=RigVMDispatchNode(False=0, True=@n14.RelativeTransform.Translation.X, Condition=&RigVMFunction_MathBoolAnd(B=@n19.Result, A=@n20.Result).Result) {template="DISPATCH_RigVMDispatch_If(in Condition,in True,in False,out Result)"}.Result
   )
   n22 = &RigUnit_GetRelativeTransformForItem(Child=(Type=Bone,Name="foot_r"), Parent=(Type=Bone,Name="root"), bChildInitial=False, bParentInitial=false)
   n23 = &RigVMFunction_MathBoolAnd(
    A=&RigVMFunction_MathDoubleGreater(B=2.5, A=@n22.RelativeTransform.Translation.X) {template="Greater::Execute(in A,in B,out Result)"}.Result
    B=&RigVMFunction_MathDoubleLess(B=-22, A=@n22.RelativeTransform.Translation.Y) {template="Less::Execute(in A,in B,out Result)"}.Result
   )
   Subtract = &RigVMFunction_MathDoubleSub(
    A=&RigVMFunction_MathDoubleClamp(Value=&RigVMFunction_MathDoubleAdd(A=@n18.Result, B=@n21.Result).Result).Result
    B=&RigVMFunction_MathDoubleRemap(
     SourceMaximum=40
     SourceMinimum=10
     TargetMaximum=1
     TargetMinimum=0
     Value=RigVMDispatchNode(False=0, True=@n22.RelativeTransform.Translation.X, Condition=@n23.Result) {template="DISPATCH_RigVMDispatch_If(in Condition,in True,in False,out Result)"}.Result
    ).Result
   )
   KalmanFilter = &RigVMFunction_KalmanFloat(BufferSize=2, Value=@Subtract.Result) {template="KalmanFilter::Execute(in Value,in BufferSize,out Result)"}
   > RigVMVariableNode(Value=@KalmanFilter.Result) {variable=L Foot Across Rear}
   RigUnit_GetRelativeTransformForItem = &RigUnit_GetRelativeTransformForItem(Child=(Type=Bone,Name="foot_r"), Parent=(Type=Bone,Name="root"), bChildInitial=False, bParentInitial=false)
   n24 = &RigVMFunction_MathBoolAnd(
    B=&RigVMFunction_MathDoubleGreater(B=-5, A=@RigUnit_GetRelativeTransformForItem.RelativeTransform.Translation.Y) {template="Greater::Execute(in A,in B,out Result)"}.Result
    A=&RigVMFunction_MathDoubleGreater(B=0, A=@RigUnit_GetRelativeTransformForItem.RelativeTransform.Translation.X) {template="Greater::Execute(in A,in B,out Result)"}.Result
   )
   RigVMFunction_AnimEvalRichCurve = &RigVMFunction_AnimEvalRichCurve(
    Curve=(EditorCurveData=(Keys=((),(Time=0.1076,Value=0.5162),(Time=1,Value=1))))
    SourceMinimum=0
    TargetMinimum=0
    Value=&RigVMFunction_MathDoubleRemap(
     SourceMaximum=0
     SourceMinimum=35
     TargetMaximum=0
     TargetMinimum=1
     Value=RigVMDispatchNode(False=0, True=@RigUnit_GetRelativeTransformForItem.RelativeTransform.Translation.X, Condition=@n24.Result) {template="DISPATCH_RigVMDispatch_If(in Condition,in True,in False,out Result)"}.Result
    ).Result
   )
   > RigVMVariableNode(Value=&RigVMFunction_KalmanFloat(BufferSize=2, Value=@RigVMFunction_AnimEvalRichCurve.Result) {template="KalmanFilter::Execute(in Value,in BufferSize,out Result)"}.Result) {variable=R Foot Across Front}
   n25 = RigVMDispatchNode(
    False=0
    True=@RigUnit_GetRelativeTransformForItem.RelativeTransform.Translation.X
    Condition=&RigVMFunction_MathBoolAnd(
     A=&RigVMFunction_MathDoubleGreater(B=2.5, A=@RigUnit_GetRelativeTransformForItem.RelativeTransform.Translation.X) {template="Greater::Execute(in A,in B,out Result)"}.Result
     B=&RigVMFunction_MathDoubleLess(B=-22, A=@RigUnit_GetRelativeTransformForItem.RelativeTransform.Translation.Y) {template="Less::Execute(in A,in B,out Result)"}.Result
    ).Result
   ) {template="DISPATCH_RigVMDispatch_If(in Condition,in True,in False,out Result)"}
   n26 = RigVMDispatchNode(
    False=0
    True=@RigUnit_GetRelativeTransformForItem.RelativeTransform.Translation.X
    Condition=&RigVMFunction_MathBoolAnd(
     A=&RigVMFunction_MathDoubleGreater(B=2.5, A=@RigUnit_GetRelativeTransformForItem.RelativeTransform.Translation.X) {template="Greater::Execute(in A,in B,out Result)"}.Result
     B=&RigVMFunction_MathDoubleLess(B=-5, A=@RigUnit_GetRelativeTransformForItem.RelativeTransform.Translation.Y) {template="Less::Execute(in A,in B,out Result)"}.Result
    ).Result
   ) {template="DISPATCH_RigVMDispatch_If(in Condition,in True,in False,out Result)"}
   n27 = &RigVMFunction_MathDoubleClamp(Value=&RigVMFunction_MathDoubleAdd(
     A=&RigVMFunction_MathDoubleRemap(SourceMaximum=40, SourceMinimum=10, TargetMaximum=1, TargetMinimum=0, Value=@n25.Result).Result
     B=&RigVMFunction_MathDoubleRemap(SourceMaximum=40, SourceMinimum=10, TargetMaximum=1, TargetMinimum=0, Value=@n26.Result).Result
    ).Result)
   n28 = &RigUnit_GetRelativeTransformForItem(Child=(Type=Bone,Name="foot_l"), Parent=(Type=Bone,Name="root"), bChildInitial=False, bParentInitial=False)
   n29 = &RigVMFunction_MathDoubleLess(B=-2.5, A=@n28.RelativeTransform.Translation.X) {template="Less::Execute(in A,in B,out Result)"}
   n30 = &RigVMFunction_MathDoubleLess(B=-22, A=@n28.RelativeTransform.Translation.Y) {template="Less::Execute(in A,in B,out Result)"}
   n31 = &RigVMFunction_MathDoubleRemap(
    SourceMaximum=-10
    SourceMinimum=-40
    TargetMaximum=0
    TargetMinimum=1
    Value=RigVMDispatchNode(False=0, True=@n28.RelativeTransform.Translation.X, Condition=&RigVMFunction_MathBoolAnd(A=@n29.Result, B=@n30.Result).Result) {template="DISPATCH_RigVMDispatch_If(in Condition,in True,in False,out Result)"}.Result
   )
   > RigVMVariableNode(Value=&RigVMFunction_KalmanFloat(BufferSize=2, Value=&RigVMFunction_MathDoubleSub(A=@n27.Result, B=@n31.Result).Result) {template="KalmanFilter::Execute(in Value,in BufferSize,out Result)"}.Result) {variable=R Foot Across Rear}
   Modify Transforms with Mirror = RigVMFunctionReferenceNode(
    Base Modify Data="((Item=(Type=Bone,Name=\"dyn_front_robe_l_MSTR\"),Transform=(Rotation=(X=0,Y=0,Z=-0.0349,W=0.9994),Translation=(X=0,Y=-0.5,Z=0),Scale3D=(X=1,Y=1,Z=1))),(Item=(Type=Bone,Name=\"dyn_side_robe_l_MSTR\"),Transform=(Rotation=(X=0,Y=0,Z=-0.0349,W=0.9994),Translation ...(truncated)"
    Mirror Modify Data="((Item=(Type=Bone,Name=\"dyn_front_robe_r_MSTR\"),Transform=(Rotation=(X=0,Y=0,Z=0,W=1),Translation=(X=0,Y=0,Z=0),Scale3D=(X=1,Y=1,Z=1))),(Item=(Type=Bone,Name=\"dyn_side_robe_r_MSTR\"),Transform=(Rotation=(X=0,Y=0,Z=0,W=1),Translation=(X=0,Y=0,Z=0),Scale3D=(X ...(truncated)"
    "Mirror Weight "=1
    Weight=1
    ExecuteContext=@RigVMFunction_Sequence.E
   )
   > &RigUnit_ModifyTransforms(
    ItemToModify="((Item=(Type=Bone,Name=\"dyn_front_robe_c_MSTR\"),Transform=(Rotation=(X=0,Y=0,Z=-0.0087,W=1),Translation=(X=0,Y=-0.5,Z=0),Scale3D=(X=1,Y=1,Z=1))),(Item=(Type=Bone,Name=\"dyn_rear_robe_c_MSTR\"),Transform=(Rotation=(X=0,Y=0,Z=-0.0087,W=1),Translation=(X=0,Y=-0 ...(truncated)"
    Weight=1
    WeightMinimum=0
   )
   RigVMFunctionReturnNode(ExecuteContext=@RigVMFunction_Sequence.F)
   Remap Mid Chain = RigVMFunctionReferenceNode(
    Max Value=0.7
    Min Value=0.3
    Smoothing=4
    Left=RigVMVariableNode() {variable=Thigh_l_forward_Standing}.Value
    Right=RigVMVariableNode() {variable=Thigh_r_forward_Standing}.Value
    ExecuteContext=@RigVMFunction_Sequence.C
   )
   > &RigUnit_ParentConstraint(
    AdvancedSettings=(InterpolationType=Shortest,RotationOrderForFilter=XZY)
    Child=(Type=Bone,Name="dyn_front_robe_c_01")
    Filter=(TranslationFilter=(bX=true,bY=true,bZ=true),RotationFilter=(bX=True,bY=True,bZ=True),ScaleFilter=(bX=True,bY=True,bZ=True))
    Parents=((Item=(Type=Bone,Name="dyn_front_robe_l_01"),Weight=0.5),(Item=(Type=Bone,Name="dyn_front_robe_r_01"),Weight=0.5))
    Weight=1
    Parents.1.Weight=@"Remap Mid Chain".Out Right
    Parents.0.Weight=@"Remap Mid Chain".Out Left
   )
   > &RigUnit_ParentConstraint(
    AdvancedSettings=(InterpolationType=Shortest,RotationOrderForFilter=XZY)
    Child=(Type=Bone,Name="dyn_front_robe_c_02")
    Filter=(TranslationFilter=(bX=true,bY=true,bZ=true),RotationFilter=(bX=True,bY=True,bZ=True),ScaleFilter=(bX=True,bY=True,bZ=True))
    Parents=((Item=(Type=Bone,Name="dyn_front_robe_l_02"),Weight=0.5),(Item=(Type=Bone,Name="dyn_front_robe_r_02"),Weight=0.5))
    Weight=1
    Parents.1.Weight=@"Remap Mid Chain".Out Right
    Parents.0.Weight=@"Remap Mid Chain".Out Left
   )
   > &RigUnit_ParentConstraint(
    AdvancedSettings=(InterpolationType=Shortest,RotationOrderForFilter=XZY)
    Child=(Type=Bone,Name="dyn_front_robe_c_03")
    Filter=(TranslationFilter=(bX=True,bY=true,bZ=True),RotationFilter=(bX=True,bY=True,bZ=True),ScaleFilter=(bX=True,bY=True,bZ=True))
    Parents=((Item=(Type=Bone,Name="dyn_front_robe_l_03"),Weight=0.5),(Item=(Type=Bone,Name="dyn_front_robe_r_03"),Weight=0.5))
    Weight=1
    Parents.1.Weight=@"Remap Mid Chain".Out Right
    Parents.0.Weight=@"Remap Mid Chain".Out Left
   )
   > &RigUnit_ParentConstraint(
    AdvancedSettings=(InterpolationType=Shortest,RotationOrderForFilter=XZY)
    Child=(Type=Bone,Name="dyn_front_robe_c_04")
    Filter=(TranslationFilter=(bX=True,bY=true,bZ=True),RotationFilter=(bX=True,bY=True,bZ=True),ScaleFilter=(bX=True,bY=True,bZ=True))
    Parents=((Item=(Type=Bone,Name="dyn_front_robe_l_04"),Weight=0.5),(Item=(Type=Bone,Name="dyn_front_robe_r_04"),Weight=0.5))
    Weight=1
    Parents.1.Weight=@"Remap Mid Chain".Out Right
    Parents.0.Weight=@"Remap Mid Chain".Out Left
   )
   > &RigUnit_ParentConstraint(
    AdvancedSettings=(InterpolationType=Shortest,RotationOrderForFilter=XZY)
    Child=(Type=Bone,Name="dyn_front_robe_c_05")
    Filter=(TranslationFilter=(bX=True,bY=true,bZ=True),RotationFilter=(bX=True,bY=True,bZ=True),ScaleFilter=(bX=True,bY=True,bZ=True))
    Parents=((Item=(Type=Bone,Name="dyn_front_robe_l_05"),Weight=0.5),(Item=(Type=Bone,Name="dyn_front_robe_r_05"),Weight=0.5))
    Weight=1
    Parents.1.Weight=@"Remap Mid Chain".Out Right
    Parents.0.Weight=@"Remap Mid Chain".Out Left
   )
   n32 = &RigVMFunction_MathDoubleRemap(SourceMaximum=0.7, SourceMinimum=0.3, TargetMaximum=1, TargetMinimum=0, Value=@"Remap Mid Chain".Out Left)
   > &RigUnit_ModifyTransforms(
    ItemToModify=((Item=(Type=Bone,Name="dyn_front_robe_c_01"),Transform=(Rotation=(X=0,Y=0,Z=0,W=1),Translation=(X=0,Y=0,Z=0),Scale3D=(X=1,Y=1,Z=1))))
    WeightMinimum=0
    Weight=@n32.Result
   )
   > &RigUnit_ModifyTransforms(
    ItemToModify=((Item=(Type=Bone,Name="dyn_front_robe_c_02"),Transform=(Rotation=(X=0,Y=0,Z=0,W=1),Translation=(X=0,Y=0,Z=0),Scale3D=(X=1,Y=1,Z=1))))
    WeightMinimum=0
    Weight=@n32.Result
   )
   > &RigUnit_ModifyTransforms(
    ItemToModify=((Item=(Type=Bone,Name="dyn_front_robe_c_03"),Transform=(Rotation=(X=-0.3007,Y=0,Z=0,W=0.9537),Translation=(X=0,Y=0,Z=0),Scale3D=(X=1,Y=1,Z=1))))
    WeightMinimum=0
    Weight=@n32.Result
   )
   > &RigUnit_ModifyTransforms(
    ItemToModify=((Item=(Type=Bone,Name="dyn_front_robe_c_04"),Transform=(Rotation=(X=-0.3007,Y=0,Z=0,W=0.9537),Translation=(X=0,Y=0,Z=0),Scale3D=(X=1,Y=1,Z=1))))
    WeightMinimum=0
    Weight=@n32.Result
   )
   > &RigUnit_ModifyTransforms(
    ItemToModify=((Item=(Type=Bone,Name="dyn_front_robe_c_05"),Transform=(Rotation=(X=-0.3007,Y=0,Z=0,W=0.9537),Translation=(X=0,Y=0,Z=0),Scale3D=(X=1,Y=1,Z=1))))
    WeightMinimum=0
    Weight=@n32.Result
   )
   n33 = &RigVMFunction_MathDoubleRemap(SourceMaximum=0.7, SourceMinimum=0.3, TargetMaximum=1, TargetMinimum=0, Value=@"Remap Mid Chain".Out Right)
   > &RigUnit_ModifyTransforms(
    ItemToModify=((Item=(Type=Bone,Name="dyn_front_robe_c_01"),Transform=(Rotation=(X=0,Y=0,Z=0,W=1),Translation=(X=0,Y=0,Z=0),Scale3D=(X=1,Y=1,Z=1))))
    WeightMinimum=0
    Weight=@n33.Result
   )
   > &RigUnit_ModifyTransforms(
    ItemToModify=((Item=(Type=Bone,Name="dyn_front_robe_c_02"),Transform=(Rotation=(X=0,Y=0,Z=0,W=1),Translation=(X=0,Y=0,Z=0),Scale3D=(X=1,Y=1,Z=1))))
    WeightMinimum=0
    Weight=@n33.Result
   )
   > &RigUnit_ModifyTransforms(
    ItemToModify=((Item=(Type=Bone,Name="dyn_front_robe_c_03"),Transform=(Rotation=(X=0.3007,Y=0,Z=0,W=0.9537),Translation=(X=0,Y=0,Z=0),Scale3D=(X=1,Y=1,Z=1))))
    WeightMinimum=0
    Weight=@n33.Result
   )
   > &RigUnit_ModifyTransforms(
    ItemToModify=((Item=(Type=Bone,Name="dyn_front_robe_c_04"),Transform=(Rotation=(X=0.3007,Y=0,Z=0,W=0.9537),Translation=(X=0,Y=0,Z=0),Scale3D=(X=1,Y=1,Z=1))))
    WeightMinimum=0
    Weight=@n33.Result
   )
   > &RigUnit_ModifyTransforms(
    ItemToModify=((Item=(Type=Bone,Name="dyn_front_robe_c_05"),Transform=(Rotation=(X=0.3007,Y=0,Z=0,W=0.9537),Translation=(X=0,Y=0,Z=0),Scale3D=(X=1,Y=1,Z=1))))
    WeightMinimum=0
    Weight=@n33.Result
   )
   > Remap Mid Chain_3 = RigVMFunctionReferenceNode(
    Max Value=0.8
    Min Value=0.2
    Smoothing=4
    Left=RigVMVariableNode() {variable=Calf_l_Bent}.Value
    Right=RigVMVariableNode() {variable=Calf_r_Bent}.Value
   )
   > Remap Mid Chain_1 = RigVMFunctionReferenceNode(
    Max Value=0.8
    Min Value=0.2
    Smoothing=2
    Left=RigVMVariableNode() {variable=Thigh_l_backward}.Value
    Right=RigVMVariableNode() {variable=Thigh_r_backward}.Value
   )
   > &RigUnit_ParentConstraint(
    AdvancedSettings=(InterpolationType=Shortest,RotationOrderForFilter=XZY)
    Child=(Type=Bone,Name="dyn_rear_robe_c_01")
    Filter=(TranslationFilter=(bX=true,bY=true,bZ=true),RotationFilter=(bX=True,bY=True,bZ=True),ScaleFilter=(bX=True,bY=True,bZ=True))
    Parents=((Item=(Type=Bone,Name="dyn_rear_robe_l_01"),Weight=0.5),(Item=(Type=Bone,Name="dyn_rear_robe_r_01"),Weight=0.5))
    Weight=1
    Parents.0.Weight=@"Remap Mid Chain_1".Out Left
    Parents.1.Weight=@"Remap Mid Chain_1".Out Right
   )
   > &RigUnit_ParentConstraint(
    AdvancedSettings=(InterpolationType=Shortest,RotationOrderForFilter=XZY)
    Child=(Type=Bone,Name="dyn_rear_robe_c_02")
    Filter=(TranslationFilter=(bX=true,bY=true,bZ=true),RotationFilter=(bX=True,bY=True,bZ=True),ScaleFilter=(bX=True,bY=True,bZ=True))
    Parents=((Item=(Type=Bone,Name="dyn_rear_robe_l_02"),Weight=0.5),(Item=(Type=Bone,Name="dyn_rear_robe_r_02"),Weight=0.5))
    Weight=1
    Parents.0.Weight=@"Remap Mid Chain_1".Out Left
    Parents.1.Weight=@"Remap Mid Chain_1".Out Right
   )
   > &RigUnit_ParentConstraint(
    AdvancedSettings=(InterpolationType=Shortest,RotationOrderForFilter=XZY)
    Child=(Type=Bone,Name="dyn_rear_robe_c_03")
    Filter=(TranslationFilter=(bX=true,bY=true,bZ=true),RotationFilter=(bX=True,bY=True,bZ=True),ScaleFilter=(bX=True,bY=True,bZ=True))
    Parents=((Item=(Type=Bone,Name="dyn_rear_robe_l_03"),Weight=0.5),(Item=(Type=Bone,Name="dyn_rear_robe_r_03"),Weight=0.5))
    Weight=1
    Parents.0.Weight=@"Remap Mid Chain_1".Out Left
    Parents.1.Weight=@"Remap Mid Chain_1".Out Right
   )
   > &RigUnit_ParentConstraint(
    AdvancedSettings=(InterpolationType=Shortest,RotationOrderForFilter=XZY)
    Child=(Type=Bone,Name="dyn_rear_robe_c_04")
    Filter=(TranslationFilter=(bX=true,bY=true,bZ=true),RotationFilter=(bX=True,bY=True,bZ=True),ScaleFilter=(bX=True,bY=True,bZ=True))
    Parents=((Item=(Type=Bone,Name="dyn_rear_robe_l_04"),Weight=0.5),(Item=(Type=Bone,Name="dyn_rear_robe_r_04"),Weight=0.5))
    Weight=1
    Parents.0.Weight=@"Remap Mid Chain_1".Out Left
    Parents.1.Weight=@"Remap Mid Chain_1".Out Right
   )
   > &RigUnit_ParentConstraint(
    AdvancedSettings=(InterpolationType=Shortest,RotationOrderForFilter=XZY)
    Child=(Type=Bone,Name="dyn_rear_robe_c_05")
    Filter=(TranslationFilter=(bX=true,bY=true,bZ=true),RotationFilter=(bX=True,bY=True,bZ=True),ScaleFilter=(bX=True,bY=True,bZ=True))
    Parents=((Item=(Type=Bone,Name="dyn_rear_robe_l_05"),Weight=0.5),(Item=(Type=Bone,Name="dyn_rear_robe_r_05"),Weight=0.5))
    Weight=1
    Parents.0.Weight=@"Remap Mid Chain_1".Out Left
    Parents.1.Weight=@"Remap Mid Chain_1".Out Right
   )
   > &RigUnit_ModifyTransforms(
    ItemToModify=((Item=(Type=Bone,Name="dyn_rear_robe_c_01"),Transform=(Rotation=(X=0,Y=0,Z=0,W=1),Translation=(X=0,Y=0,Z=0),Scale3D=(X=1,Y=1,Z=1))))
    WeightMinimum=0
    Weight=@"Remap Mid Chain_1".Out Left
   )
   > &RigUnit_ModifyTransforms(
    ItemToModify=((Item=(Type=Bone,Name="dyn_rear_robe_c_02"),Transform=(Rotation=(X=0,Y=0,Z=0,W=1),Translation=(X=0,Y=0,Z=0),Scale3D=(X=1,Y=1,Z=1))))
    WeightMinimum=0
    Weight=@"Remap Mid Chain_1".Out Left
   )
   > &RigUnit_ModifyTransforms(
    ItemToModify=((Item=(Type=Bone,Name="dyn_rear_robe_c_03"),Transform=(Rotation=(X=0.3007,Y=0,Z=0,W=0.9537),Translation=(X=0,Y=0,Z=0),Scale3D=(X=1,Y=1,Z=1))))
    WeightMinimum=0
    Weight=@"Remap Mid Chain_1".Out Left
   )
   > &RigUnit_ModifyTransforms(
    ItemToModify=((Item=(Type=Bone,Name="dyn_rear_robe_c_04"),Transform=(Rotation=(X=0.3007,Y=0,Z=0,W=0.9537),Translation=(X=0,Y=0,Z=0),Scale3D=(X=1,Y=1,Z=1))))
    WeightMinimum=0
    Weight=@"Remap Mid Chain_1".Out Left
   )
   > &RigUnit_ModifyTransforms(
    ItemToModify=((Item=(Type=Bone,Name="dyn_rear_robe_c_05"),Transform=(Rotation=(X=0.3007,Y=0,Z=0,W=0.9537),Translation=(X=0,Y=0,Z=0),Scale3D=(X=1,Y=1,Z=1))))
    WeightMinimum=0
    Weight=@"Remap Mid Chain_1".Out Left
   )
   > &RigUnit_ModifyTransforms(
    ItemToModify=((Item=(Type=Bone,Name="dyn_rear_robe_c_01"),Transform=(Rotation=(X=0,Y=0,Z=0,W=1),Translation=(X=0,Y=0,Z=0),Scale3D=(X=1,Y=1,Z=1))))
    WeightMinimum=0
    Weight=@"Remap Mid Chain_1".Out Right
   )
   > &RigUnit_ModifyTransforms(
    ItemToModify=((Item=(Type=Bone,Name="dyn_rear_robe_c_02"),Transform=(Rotation=(X=0,Y=0,Z=0,W=1),Translation=(X=0,Y=0,Z=0),Scale3D=(X=1,Y=1,Z=1))))
    WeightMinimum=0
    Weight=@"Remap Mid Chain_1".Out Right
   )
   > &RigUnit_ModifyTransforms(
    ItemToModify=((Item=(Type=Bone,Name="dyn_rear_robe_c_03"),Transform=(Rotation=(X=-0.3007,Y=0,Z=0,W=0.9537),Translation=(X=0,Y=0,Z=0),Scale3D=(X=1,Y=1,Z=1))))
    WeightMinimum=0
    Weight=@"Remap Mid Chain_1".Out Right
   )
   > &RigUnit_ModifyTransforms(
    ItemToModify=((Item=(Type=Bone,Name="dyn_rear_robe_c_04"),Transform=(Rotation=(X=-0.3007,Y=0,Z=0,W=0.9537),Translation=(X=0,Y=0,Z=0),Scale3D=(X=1,Y=1,Z=1))))
    WeightMinimum=0
    Weight=@"Remap Mid Chain_1".Out Right
   )
   > &RigUnit_ModifyTransforms(
    ItemToModify=((Item=(Type=Bone,Name="dyn_rear_robe_c_05"),Transform=(Rotation=(X=-0.3007,Y=0,Z=0,W=0.9537),Translation=(X=0,Y=0,Z=0),Scale3D=(X=1,Y=1,Z=1))))
    WeightMinimum=0
    Weight=@"Remap Mid Chain_1".Out Right
   )
   Add = &RigVMFunction_MathDoubleAdd(A=RigVMVariableNode() {variable=Thigh_l_backward}.Value, B=RigVMVariableNode() {variable=Thigh_r_backward}.Value)
   Clamp = &RigVMFunction_MathDoubleClamp(Value=@Add.Result)
   > &RigUnit_ModifyTransforms(
    ItemToModify=((Item=(Type=Bone,Name="dyn_rear_robe_c_MSTR"),Transform=(Rotation=(X=0,Y=0,Z=-0.0218,W=0.9998),Translation=(X=0,Y=-1,Z=0),Scale3D=(X=1,Y=1,Z=1))))
    WeightMinimum=0
    Weight=@Clamp.Result
   )
   > &RigUnit_ModifyTransforms(
    ItemToModify=((Item=(Type=Bone,Name="dyn_rear_robe_c_02"),Transform=(Rotation=(X=0,Y=0,Z=-0.0218,W=0.9998),Translation=(X=0,Y=-2.5,Z=0),Scale3D=(X=1,Y=1,Z=1))))
    WeightMinimum=0
    Weight=@Clamp.Result
   )
   &RigUnit_ParentConstraint(
    AdvancedSettings=(InterpolationType=Average,RotationOrderForFilter=XZY)
    Child=(Type=Bone,Name="dyn_front_robe_l_03")
    Filter=(TranslationFilter=(bX=True,bY=true,bZ=true),RotationFilter=(bX=true,bY=true,bZ=true),ScaleFilter=(bX=True,bY=True,bZ=True))
    Parents=((Item=(Type=Bone,Name="calf_l"),Weight=1))
    Weight=1
    ExecutePin=@RigVMFunction_Sequence.B
   )
   > &RigUnit_ParentConstraint(
    AdvancedSettings=(InterpolationType=Average,RotationOrderForFilter=XZY)
    Child=(Type=Bone,Name="dyn_rear_robe_l_03")
    Filter=(TranslationFilter=(bX=true,bY=true,bZ=true),RotationFilter=(bX=true,bY=true,bZ=true),ScaleFilter=(bX=True,bY=True,bZ=True))
    Parents=((Item=(Type=Bone,Name="calf_l"),Weight=1))
    Weight=1
   )
   > &RigUnit_ParentConstraint(
    AdvancedSettings=(InterpolationType=Average,RotationOrderForFilter=XZY)
    Child=(Type=Bone,Name="dyn_side_robe_l_03")
    Filter=(TranslationFilter=(bX=True,bY=true,bZ=true),RotationFilter=(bX=true,bY=true,bZ=true),ScaleFilter=(bX=True,bY=True,bZ=True))
    Parents=((Item=(Type=Bone,Name="calf_l"),Weight=1))
    Weight=1
   )
   > &RigUnit_ParentConstraint(
    AdvancedSettings=(InterpolationType=Average,RotationOrderForFilter=XZY)
    Child=(Type=Bone,Name="dyn_front_robe_r_03")
    Filter=(TranslationFilter=(bX=True,bY=True,bZ=True),RotationFilter=(bX=true,bY=true,bZ=true),ScaleFilter=(bX=True,bY=True,bZ=True))
    Parents=((Item=(Type=Bone,Name="calf_r"),Weight=1))
    Weight=1
   )
   > &RigUnit_ParentConstraint(
    AdvancedSettings=(InterpolationType=Average,RotationOrderForFilter=XZY)
    Child=(Type=Bone,Name="dyn_rear_robe_r_03")
    Filter=(TranslationFilter=(bX=true,bY=true,bZ=true),RotationFilter=(bX=true,bY=true,bZ=true),ScaleFilter=(bX=True,bY=True,bZ=True))
    Parents=((Item=(Type=Bone,Name="calf_r"),Weight=1))
    Weight=1
   )
   > n34 = &RigUnit_ParentConstraint(
    AdvancedSettings=(InterpolationType=Average,RotationOrderForFilter=XZY)
    Child=(Type=Bone,Name="dyn_side_robe_r_03")
    Filter=(TranslationFilter=(bX=true,bY=true,bZ=true),RotationFilter=(bX=true,bY=true,bZ=true),ScaleFilter=(bX=True,bY=True,bZ=True))
    Parents=((Item=(Type=Bone,Name="calf_r"),Weight=1))
    Weight=1
   )
   RigVMFunction_ControlFlowBranch = &RigVMFunction_ControlFlowBranch(
    Condition=&RigVMFunction_MathDoubleGreater(B=0, A=RigVMVariableNode() {variable=Is Not Emote Alpha}.Value) {template="Greater::Execute(in A,in B,out Result)"}.Result
    ExecuteContext=@n34.ExecutePin
   )
   n35 = RigVMVariableNode() {variable=L Foot Across Front}
   &RigUnit_ParentConstraint(
    AdvancedSettings=(InterpolationType=Average,RotationOrderForFilter=XZY)
    Child=(Type=Bone,Name="dyn_front_robe_r_MSTR")
    Parents=((Item=(Type=Bone,Name="deform_thigh_l"),Weight=1))
    ExecutePin=@RigVMFunction_ControlFlowBranch.True
    Weight=@n35.Value
   )
   > &RigUnit_ParentConstraint(
    AdvancedSettings=(InterpolationType=Average,RotationOrderForFilter=XZY)
    Child=(Type=Bone,Name="dyn_front_robe_r_03")
    Filter=(TranslationFilter=(bX=true,bY=true,bZ=true),RotationFilter=(bX=True,bY=True,bZ=True),ScaleFilter=(bX=True,bY=True,bZ=True))
    Parents=((Item=(Type=Bone,Name="calf_l"),Weight=1))
    Weight=@n35.Value
   )
   Multiply_8 = RigVMAggregateNode(
    A=RigVMVariableNode() {variable=L Foot Across Rear}.Value
    B=RigVMVariableNode() {variable=Is Skydiving Alpha}.Value
    C=RigVMVariableNode() {variable=R Foot Across Front}.Value
   )
   > &RigUnit_ParentConstraint(
    AdvancedSettings=(InterpolationType=Average,RotationOrderForFilter=XZY)
    Child=(Type=Bone,Name="dyn_rear_robe_r_MSTR")
    Filter=(TranslationFilter=(bX=true,bY=true,bZ=true),RotationFilter=(bX=True,bY=True,bZ=True),ScaleFilter=(bX=True,bY=True,bZ=True))
    Parents=((Item=(Type=Bone,Name="deform_thigh_l"),Weight=1))
    Weight=@Multiply_8.Result
   )
   > &RigUnit_ParentConstraint(
    AdvancedSettings=(InterpolationType=Average,RotationOrderForFilter=XZY)
    Child=(Type=Bone,Name="dyn_rear_robe_r_03")
    Filter=(TranslationFilter=(bX=true,bY=true,bZ=true),RotationFilter=(bX=True,bY=True,bZ=True),ScaleFilter=(bX=True,bY=True,bZ=True))
    Parents=((Item=(Type=Bone,Name="calf_l"),Weight=1))
    Weight=@Multiply_8.Result
   )
   n36 = RigVMVariableNode() {variable=R Foot Across Front}
   > &RigUnit_ParentConstraint(
    AdvancedSettings=(InterpolationType=Average,RotationOrderForFilter=XZY)
    Child=(Type=Bone,Name="dyn_front_robe_l_MSTR")
    Parents=((Item=(Type=Bone,Name="deform_thigh_r"),Weight=1))
    Weight=@n36.Value
   )
   > &RigUnit_ParentConstraint(
    AdvancedSettings=(InterpolationType=Average,RotationOrderForFilter=XZY)
    Child=(Type=Bone,Name="dyn_front_robe_l_03")
    Parents=((Item=(Type=Bone,Name="calf_r"),Weight=1))
    Weight=@n36.Value
   )
   Multiply_9 = RigVMAggregateNode(
    A=RigVMVariableNode() {variable=R Foot Across Rear}.Value
    C=RigVMVariableNode() {variable=L Foot Across Front}.Value
    B=RigVMVariableNode() {variable=Is Skydiving Alpha}.Value
   )
   > &RigUnit_ParentConstraint(
    AdvancedSettings=(InterpolationType=Average,RotationOrderForFilter=XZY)
    Child=(Type=Bone,Name="dyn_rear_robe_l_MSTR")
    Parents=((Item=(Type=Bone,Name="deform_thigh_r"),Weight=1))
    Weight=@Multiply_9.Result
   )
   > &RigUnit_ParentConstraint(
    AdvancedSettings=(InterpolationType=Average,RotationOrderForFilter=XZY)
    Child=(Type=Bone,Name="dyn_rear_robe_l_03")
    Parents=((Item=(Type=Bone,Name="calf_r"),Weight=1))
    Weight=@Multiply_9.Result
   )
   Multiply = &RigVMFunction_MathDoubleMul(B=0.5, A=RigVMVariableNode() {variable=R Foot Across Rear}.Value)
   > &RigUnit_ParentConstraint(
    AdvancedSettings=(InterpolationType=Average,RotationOrderForFilter=XZY)
    Child=(Type=Bone,Name="dyn_front_robe_r_MSTR")
    Parents=((Item=(Type=Bone,Name="deform_thigh_l"),Weight=1))
    Weight=@Multiply.Result
   )
   > &RigUnit_ParentConstraint(
    AdvancedSettings=(InterpolationType=Average,RotationOrderForFilter=XZY)
    Child=(Type=Bone,Name="dyn_front_robe_l_MSTR")
    Parents=((Item=(Type=Bone,Name="deform_thigh_r"),Weight=1))
    Weight=&RigVMFunction_MathDoubleMul(B=0.5, A=RigVMVariableNode() {variable=L Foot Across Rear}.Value).Result
   )
   > &RigUnit_ParentConstraint(
    AdvancedSettings=(InterpolationType=Average,RotationOrderForFilter=XZY)
    Child=(Type=Bone,Name="dyn_side_robe_r_MSTR")
    Parents=((Item=(Type=Bone,Name="deform_thigh_l"),Weight=0.5),(Item=(Type=Bone,Name="deform_thigh_r"),Weight=0.5))
    Weight=&RigVMFunction_MathDoubleMul(B=0.5, A=RigVMVariableNode() {variable=R Foot Across Rear}.Value).Result
   )
   > &RigUnit_ParentConstraint(
    AdvancedSettings=(InterpolationType=Average,RotationOrderForFilter=XZY)
    Child=(Type=Bone,Name="dyn_side_robe_l_MSTR")
    Parents=((Item=(Type=Bone,Name="deform_thigh_r"),Weight=0.5),(Item=(Type=Bone,Name="deform_thigh_l"),Weight=0.5))
    Weight=&RigVMFunction_MathDoubleMul(B=0.5, A=RigVMVariableNode() {variable=L Foot Across Rear}.Value).Result
   )
   > &RigUnit_ModifyTransforms(
    ItemToModify="((Item=(Type=Bone,Name=\"dyn_front_robe_l_MSTR\"),Transform=(Rotation=(X=0,Y=0,Z=0,W=1),Translation=(X=0,Y=0,Z=0),Scale3D=(X=1,Y=1,Z=1))),(Item=(Type=Bone,Name=\"dyn_front_robe_c_MSTR\"),Transform=(Rotation=(X=0.0627,Y=0.1755,Z=0.016,W=-0.9824),Translation=(X= ...(truncated)"
    WeightMinimum=0
    Weight=RigVMVariableNode() {variable=L Foot Across Front}.Value
   )
   > &RigUnit_ModifyTransforms(
    ItemToModify="((Item=(Type=Bone,Name=\"dyn_rear_robe_l_MSTR\"),Transform=(Rotation=(X=0,Y=0,Z=0,W=1),Translation=(X=0,Y=0,Z=0),Scale3D=(X=1,Y=1,Z=1))),(Item=(Type=Bone,Name=\"dyn_rear_robe_c_MSTR\"),Transform=(Rotation=(X=0,Y=-0.0649,Z=0,W=0.9979),Translation=(X=0,Y=0,Z=0), ...(truncated)"
    WeightMinimum=0
    Weight=&RigVMFunction_MathDoubleMul(A=RigVMVariableNode() {variable=L Foot Across Rear}.Value, B=RigVMVariableNode() {variable=R Foot Across Front}.Value).Result
   )
   > &RigUnit_ModifyTransforms(
    ItemToModify="((Item=(Type=Bone,Name=\"dyn_front_robe_l_MSTR\"),Transform=(Rotation=(X=0.1367,Y=0.018,Z=0.1293,W=0.982),Translation=(X=0,Y=0,Z=0),Scale3D=(X=1,Y=1,Z=1))),(Item=(Type=Bone,Name=\"dyn_front_robe_c_MSTR\"),Transform=(Rotation=(X=0,Y=0,Z=0,W=1),Translation=(X=0, ...(truncated)"
    WeightMinimum=0
    Weight=RigVMVariableNode() {variable=R Foot Across Front}.Value
   )
   > n37 = &RigUnit_ModifyTransforms(
    ItemToModify="((Item=(Type=Bone,Name=\"dyn_rear_robe_l_MSTR\"),Transform=(Rotation=(X=0,Y=-0.0436,Z=0,W=0.999),Translation=(X=0,Y=0,Z=0),Scale3D=(X=1,Y=1,Z=1))),(Item=(Type=Bone,Name=\"dyn_rear_robe_c_MSTR\"),Transform=(Rotation=(X=0,Y=-0.0649,Z=0,W=0.9979),Translation=(X=0 ...(truncated)"
    WeightMinimum=0
    Weight=&RigVMFunction_MathDoubleMul(A=RigVMVariableNode() {variable=R Foot Across Rear}.Value, B=RigVMVariableNode() {variable=L Foot Across Front}.Value).Result
   )
   n38 = &RigVMFunction_ControlFlowBranch(ExecuteContext=@n37.ExecutePin, Condition=RigVMVariableNode() {variable=IsSkydiving}.Value)
   &RigUnit_ModifyTransforms(
    ItemToModify=((Item=(Type=Bone,Name="dyn_rear_robe_c_MSTR"),Transform=(Rotation=(X=0,Y=0,Z=-0.0436,W=0.999),Translation=(X=0,Y=0,Z=0),Scale3D=(X=1,Y=1,Z=1))))
    Weight=1
    WeightMinimum=0
    ExecutePin=@n38.True
   )
   > &RigUnit_ModifyTransforms(
    ItemToModify=((Item=(Type=Bone,Name="dyn_rear_robe_l_MSTR"),Transform=(Rotation=(X=0,Y=0,Z=-0.0436,W=0.999),Translation=(X=0,Y=0,Z=0),Scale3D=(X=1,Y=1,Z=1))))
    Weight=1
    WeightMinimum=0
   )
   > &RigUnit_ModifyTransforms(
    ItemToModify=((Item=(Type=Bone,Name="dyn_rear_robe_r_MSTR"),Transform=(Rotation=(X=0,Y=0,Z=-0.0436,W=0.999),Translation=(X=0,Y=0,Z=0),Scale3D=(X=1,Y=1,Z=1))))
    Weight=1
    WeightMinimum=0
   )
   n39 = &RigVMFunction_MathDoubleClamp(Value=&RigVMFunction_MathDoubleAdd(A=RigVMVariableNode() {variable=Calf_l_Bent}.Value, B=RigVMVariableNode() {variable=Calf_r_Bent}.Value).Result)
   &RigUnit_ModifyTransforms(
    ItemToModify=((Item=(Type=Bone,Name="dyn_rear_robe_l_03"),Transform=(Rotation=(X=0,Y=0,Z=0.0436,W=0.999),Translation=(X=5,Y=-5,Z=0),Scale3D=(X=1,Y=1,Z=1))))
    WeightMinimum=0
    Weight=@n39.Result
    ExecutePin=@RigVMFunction_Sequence.D
   )
   > &RigUnit_ModifyTransforms(
    ItemToModify=((Item=(Type=Bone,Name="dyn_rear_robe_c_03"),Transform=(Rotation=(X=0,Y=0,Z=0.0436,W=0.999),Translation=(X=5,Y=-5,Z=0),Scale3D=(X=1,Y=1,Z=1))))
    WeightMinimum=0
    Weight=@n39.Result
   )
   > &RigUnit_ModifyTransforms(
    ItemToModify=((Item=(Type=Bone,Name="dyn_rear_robe_r_03"),Transform=(Rotation=(X=0,Y=0,Z=0.0436,W=0.999),Translation=(X=-5,Y=5,Z=0),Scale3D=(X=1,Y=1,Z=1))))
    WeightMinimum=0
    Weight=@n39.Result
   )
   RigVMVariableNode() {variable=OutputParam}
   RigVMVariableNode() {variable=OutputParam_0}
  ]}
 New Function_ContainedGraph {nodes [
   RigVMFunctionEntryNode()
   > RigVMFunction_Sequence = RigVMAggregateNode()
   RigVMFunction_ControlFlowBranch = &RigVMFunction_ControlFlowBranch(ExecuteContext=@RigVMFunction_Sequence.A, Condition=RigVMVariableNode() {variable=IsSkydiving}.Value)
   ItemArray = RigVMDispatchNode(Value="((Type=Bone,Name=\"dyn_side_robe_l_MSTR\"),(Type=Bone,Name=\"dyn_rear_robe_l_MSTR\"),(Type=Bone,Name=\"dyn_front_robe_l_MSTR\"),(Type=Bone,Name=\"dyn_side_robe_r_MSTR\"),(Type=Bone,Name=\"dyn_rear_robe_r_MSTR\"),(Type=Bone,Name=\"dyn_front_robe_r_MSTR\"),(Type=Bone,Na ...(truncated)") {template="DISPATCH_RigVMDispatch_Constant(io Value)"}
   For_Each = RigVMDispatchNode(Array=@ItemArray.Value, ExecuteContext=@RigVMFunction_ControlFlowBranch.True) {template="DISPATCH_RigVMDispatch_ArrayIterator(in Array,out Element,out Index,out Count,out Ratio)"}
   Remap = &RigVMFunction_MathDoubleRemap(SourceMaximum=1300, SourceMinimum=440, TargetMaximum=1, TargetMinimum=0, Value=RigVMVariableNode() {variable=PlayerSpeed}.Value)
   Multiply = &RigVMFunction_MathDoubleMul(B=0.3, A=@Remap.Result)
   > /Script/ControlRig.RigUnit_ChainHarmonicsPerItem(
    DrawWorldOffset=(Rotation=(X=0,Y=0,Z=0,W=1),Translation=(X=0,Y=0,Z=0),Scale3D=(X=1,Y=1,Z=1))
    Pendulum=(bEnabled=False,PendulumStiffness=2,PendulumGravity=(X=0,Y=0,Z=0),PendulumBlend=0.75,PendulumDrag=0.98,PendulumMinimum=0,PendulumMaximum=0.1,PendulumEase=Linear,UnwindAxis=(X=0,Y=1,Z=0),UnwindMinimum=0.2,UnwindMaximum=0.05)
    Reach=(bEnabled=False,ReachTarget=(X=0,Y=0,Z=0),ReachAxis=(X=1,Y=0,Z=0),ReachMinimum=0,ReachMaximum=0,ReachEase=Linear)
    Speed=(X=30,Y=30,Z=30)
    Wave=(bEnabled=true,WaveFrequency=(X=0,Y=0.6,Z=20),WaveAmplitude=(X=0.2,Y=0.2,Z=0.4),WaveOffset=(X=0,Y=1,Z=2),WaveNoise=(X=0.5,Y=0.5,Z=0.01),WaveMinimum=0,WaveMaximum=0.3,WaveEase=Linear)
    WaveCurve=(EditorCurveData=(Keys=((),(Time=0.5671,Value=0.3976),(Time=1,Value=1))))
    bDrawDebug=false
    ChainRoot=@For_Each.Element
    Wave.WaveMaximum=@Multiply.Result
   ) {is=struct RigVMUnitNode}
   &RigUnit_ModifyTransforms(
    ItemToModify=((Item=(Type=Bone,Name="dyn_sleeve_l_bk_03"),Transform=(Rotation=(X=0,Y=0,Z=-0.1305,W=0.9914),Translation=(X=0,Y=0,Z=0),Scale3D=(X=1,Y=1,Z=1))))
    Weight=1
    WeightMinimum=0
    ExecutePin=@For_Each.Completed
   )
   > &RigUnit_ModifyTransforms(
    ItemToModify=((Item=(Type=Bone,Name="dyn_sleeve_r_bk_03"),Transform=(Rotation=(X=0,Y=0,Z=0.1305,W=0.9914),Translation=(X=0,Y=0,Z=0),Scale3D=(X=1,Y=1,Z=1))))
    Weight=1
    WeightMinimum=0
   )
   > &RigUnit_ModifyTransforms(
    ItemToModify=((Item=(Type=Bone,Name="dyn_sleeve_l_tp_03"),Transform=(Rotation=(X=0,Y=-0.1305,Z=0,W=0.9914),Translation=(X=0,Y=0,Z=0),Scale3D=(X=1,Y=1,Z=1))))
    Weight=1
    WeightMinimum=0
   )
   > &RigUnit_ModifyTransforms(
    ItemToModify=((Item=(Type=Bone,Name="dyn_sleeve_r_tp_03"),Transform=(Rotation=(X=0,Y=-0.1305,Z=0,W=0.9914),Translation=(X=0,Y=0,Z=0),Scale3D=(X=1,Y=1,Z=1))))
    Weight=1
    WeightMinimum=0
   )
   ItemArray_1 = RigVMDispatchNode(Value="((Type=Bone,Name=\"dyn_sleeve_l_bk_03\"),(Type=Bone,Name=\"dyn_sleeve_l_bt_03\"),(Type=Bone,Name=\"dyn_sleeve_l_ft_03\"),(Type=Bone,Name=\"dyn_sleeve_l_tp_03\"),(Type=Bone,Name=\"dyn_sleeve_r_bk_03\"),(Type=Bone,Name=\"dyn_sleeve_r_bt_03\"),(Type=Bone,Name=\"dyn_sleeve ...(truncated)") {template="DISPATCH_RigVMDispatch_Constant(io Value)"}
   > For_Each_1 = RigVMDispatchNode(Array=@ItemArray_1.Value) {template="DISPATCH_RigVMDispatch_ArrayIterator(in Array,out Element,out Index,out Count,out Ratio)"}
   n1 = RigVMVariableNode() {variable=PlayerSpeed}
   > &RigUnit_ChainHarmonicsPerItem(
    DrawWorldOffset=(Rotation=(X=0,Y=0,Z=0,W=1),Translation=(X=0,Y=0,Z=0),Scale3D=(X=1,Y=1,Z=1))
    Pendulum=(bEnabled=False,PendulumStiffness=2,PendulumGravity=(X=0,Y=0,Z=0),PendulumBlend=0.75,PendulumDrag=0.98,PendulumMinimum=0,PendulumMaximum=0.1,PendulumEase=Linear,UnwindAxis=(X=0,Y=1,Z=0),UnwindMinimum=0.2,UnwindMaximum=0.05)
    Reach=(bEnabled=False,ReachTarget=(X=0,Y=0,Z=0),ReachAxis=(X=1,Y=0,Z=0),ReachMinimum=0,ReachMaximum=0,ReachEase=Linear)
    Speed=(X=15,Y=15,Z=15)
    Wave=(bEnabled=true,WaveFrequency=(X=0,Y=0.2,Z=10),WaveAmplitude=(X=0.1,Y=0.1,Z=0.2),WaveOffset=(X=0,Y=0.5,Z=1),WaveNoise=(X=0.25,Y=0.25,Z=0.01),WaveMinimum=0,WaveMaximum=0.3,WaveEase=Linear)
    WaveCurve=(EditorCurveData=(Keys=((),(Time=0.5671,Value=0.3976),(Time=1,Value=1))))
    bDrawDebug=false
    ChainRoot=@For_Each_1.Element
    Wave.WaveMaximum=&RigVMFunction_MathDoubleMul(
     B=0.3
     A=&RigVMFunction_MathDoubleRemap(SourceMaximum=1300, SourceMinimum=440, TargetMaximum=1, TargetMinimum=0, Value=@n1.Value).Result
    ).Result
   )
   RigVMFunctionReturnNode(ExecuteContext=@RigVMFunction_Sequence.C)
  ]}
 New Function_ContainedGraph {nodes [
   RigVMFunctionEntryNode()
   n1 = &RigVMFunction_MathDoubleGreater(B=RigVMVariableNode() {variable=Right}.Value, A=RigVMVariableNode() {variable=Left}.Value) {template="Greater::Execute(in A,in B,out Result)"}
   n2 = &RigVMFunction_AnimEvalRichCurve(
    Curve=(EditorCurveData=(Keys=((InterpMode=RCIM_Cubic,TangentMode=RCTM_User,ArriveTangent=2.6176,LeaveTangent=2.6177),(InterpMode=RCIM_Cubic,Time=0.3194,Value=0.6741,ArriveTangent=1,LeaveTangent=1),(Time=1,Value=1))))
    SourceMinimum=0
    TargetMinimum=0
    Value=RigVMVariableNode() {variable=Left}.Value
   )
   n3 = &RigVMFunction_MathDoubleRemap(SourceMaximum=1, SourceMinimum=0.5, TargetMaximum=1, TargetMinimum=0, Value=@n2.Result)
   n4 = &RigVMFunction_MathDoubleRemap(SourceMaximum=1, SourceMinimum=0, TargetMinimum=0.5, Value=@n3.Result, TargetMaximum=RigVMVariableNode() {variable=Max Value}.Value)
   n5 = &RigVMFunction_AnimEvalRichCurve(
    Curve=(EditorCurveData=(Keys=((InterpMode=RCIM_Cubic,TangentMode=RCTM_User,ArriveTangent=2.6176,LeaveTangent=2.6177),(InterpMode=RCIM_Cubic,Time=0.3194,Value=0.6741,ArriveTangent=1,LeaveTangent=1),(Time=1,Value=1))))
    SourceMinimum=0
    TargetMinimum=0
    Value=RigVMVariableNode() {variable=Right}.Value
   )
   n6 = &RigVMFunction_MathDoubleRemap(SourceMaximum=1, SourceMinimum=0.5, TargetMaximum=1, TargetMinimum=0, Value=@n5.Result)
   n7 = &RigVMFunction_MathDoubleRemap(SourceMaximum=1, SourceMinimum=0, TargetMinimum=0.5, Value=@n6.Result, TargetMaximum=RigVMVariableNode() {variable=Min Value}.Value)
   RigVMFunction_MathBoolAnd = &RigVMFunction_MathBoolAnd(
    A=RigVMDispatchNode(B=1, A=RigVMVariableNode() {variable=Left}.Value) {template="DISPATCH_RigVMDispatch_CoreEquals(in A,in B,out Result)"}.Result
    B=RigVMDispatchNode(B=1, A=RigVMVariableNode() {variable=Right}.Value) {template="DISPATCH_RigVMDispatch_CoreEquals(in A,in B,out Result)"}.Result
   )
   n8 = RigVMVariableNode() {variable=Left}
   n9 = RigVMVariableNode() {variable=Right}
   RigVMFunction_MathBoolOr = &RigVMFunction_MathBoolOr(
    A=@RigVMFunction_MathBoolAnd.Result
    B=&RigVMFunction_MathBoolAnd(
     A=RigVMDispatchNode(B=0, A=@n8.Value) {template="DISPATCH_RigVMDispatch_CoreEquals(in A,in B,out Result)"}.Result
     B=RigVMDispatchNode(B=0, A=@n9.Value) {template="DISPATCH_RigVMDispatch_CoreEquals(in A,in B,out Result)"}.Result
    ).Result
   )
   KalmanFilter = &RigVMFunction_KalmanFloat(
    Value=RigVMDispatchNode(
     True=0.5
     False=RigVMDispatchNode(Condition=@n1.Result, True=@n4.Result, False=@n7.Result) {template="DISPATCH_RigVMDispatch_If(in Condition,in True,in False,out Result)"}.Result
     Condition=@RigVMFunction_MathBoolOr.Result
    ) {template="DISPATCH_RigVMDispatch_If(in Condition,in True,in False,out Result)"}.Result
    BufferSize=RigVMVariableNode() {variable=Smoothing}.Value
   ) {template="KalmanFilter::Execute(in Value,in BufferSize,out Result)"}
   n10 = RigVMDispatchNode(
    Condition=@n1.Result
    True=&RigVMFunction_MathDoubleRemap(SourceMaximum=1, SourceMinimum=0, TargetMinimum=0.5, Value=@n3.Result, TargetMaximum=RigVMVariableNode() {variable=Min Value}.Value).Result
    False=&RigVMFunction_MathDoubleRemap(SourceMaximum=1, SourceMinimum=0, TargetMinimum=0.5, Value=@n6.Result, TargetMaximum=RigVMVariableNode() {variable=Max Value}.Value).Result
   ) {template="DISPATCH_RigVMDispatch_If(in Condition,in True,in False,out Result)"}
   > RigVMFunctionReturnNode(
    Out Left=@KalmanFilter.Result
    Out Right=&RigVMFunction_KalmanFloat(
     Value=RigVMDispatchNode(True=0.5, False=@n10.Result, Condition=@RigVMFunction_MathBoolOr.Result) {template="DISPATCH_RigVMDispatch_If(in Condition,in True,in False,out Result)"}.Result
     BufferSize=RigVMVariableNode() {variable=Smoothing}.Value
    ) {template="KalmanFilter::Execute(in Value,in BufferSize,out Result)"}.Result
   )
  ]}
 RigVMFunctionLibrary {nodes [
   Remap Mid Chain = RigVMCollapseNode(ExecuteContext=(), Max Value=0.8, Min Value=0.2, Smoothing=4)
   Robe Wind = RigVMCollapseNode(ExecuteContext=())
   RobeLogic_2 = RigVMCollapseNode(ExecuteContext=())
   SleeveLogic = RigVMCollapseNode(ExecuteContext=())
  ]}
 RigVMFunction_Sequence_1_ContainedGraph {nodes [
   n1 = RigVMFunctionEntryNode()
   RigVMFunction_Sequence = &RigVMFunction_Sequence(ExecuteContext=@n1.ExecuteContext)
   n2 = &RigVMFunction_Sequence(ExecuteContext=@RigVMFunction_Sequence.B)
   RigVMFunctionReturnNode(A=@RigVMFunction_Sequence.A, B=@n2.A, C=@n2.B)
  ]}
 RigVMFunction_Sequence_1_ContainedGraph {nodes [
   n1 = RigVMFunctionEntryNode()
   RigVMFunction_Sequence = &RigVMFunction_Sequence(ExecuteContext=@n1.ExecuteContext)
   n2 = &RigVMFunction_Sequence(ExecuteContext=@RigVMFunction_Sequence.B)
   n3 = &RigVMFunction_Sequence(ExecuteContext=@n2.B)
   n4 = &RigVMFunction_Sequence(ExecuteContext=@n3.B)
   n5 = &RigVMFunction_Sequence(ExecuteContext=@n4.B)
   RigVMFunctionReturnNode(A=@RigVMFunction_Sequence.A, B=@n2.A, C=@n3.A, D=@n4.A, E=@n5.A, F=@n5.B)
  ]}
 RigVMFunction_Sequence_1_ContainedGraph {nodes [
   n1 = RigVMFunctionEntryNode()
   RigVMFunction_Sequence = &RigVMFunction_Sequence(ExecuteContext=@n1.ExecuteContext)
   n2 = &RigVMFunction_Sequence(ExecuteContext=@RigVMFunction_Sequence.B)
   RigVMFunctionReturnNode(A=@RigVMFunction_Sequence.A, B=@n2.A, C=@n2.B)
  ]}
 RigVMModel {nodes [
   RigUnit_BeginExecution = &RigUnit_BeginExecution()
   /Script/FortniteGame.RigUnit_FortDeformRig(
    Bool=(bUpperArm=False,bElbow=False,bWrist=False,bGroin=False,bGlute=False,bKnee=False,bPec=False,bBicep=False,bLat=False)
    Curve="(bLUpperArmUpAlpha=False,bRUpperArmUpAlpha=False,bLUpperArmFrAlpha=False,bRUpperArmFrAlpha=False,bLUpperArmBkAlpha=False,bRUpperArmBkAlpha=False,bLWristUpAlpha=False,bRWristUpAlpha=False,bLWristDnAlpha=False,bRWristDnAlpha=False,bLKneeAlpha=False,bRKneeAlp ...(truncated)"
    Debug=(bUpperArmDebug=False,bWristDebug=False,bLatDebug=False,bGroinGluteDebug=False)
    Value="(UpperArmUp=(Rotation=(X=0.0,Y=-0.0,Z=0.0,W=1.0),Translation=(X=0.0,Y=0.0,Z=8.0),Scale3D=(X=0.70,Y=1.0,Z=1.0)),UpperArmFwdFr=(Rotation=(X=0.0,Y=-0.0,Z=0.0,W=1.0),Translation=(X=0.0,Y=10.0,Z=0.0),Scale3D=(X=0.50,Y=1.0,Z=1.0)),UpperArmFwdBk=(Rotation=(X=0.0, ...(truncated)"
    ExecutePin=@RigUnit_BeginExecution.ExecutePin
   ) {is=struct RigVMUnitNode}
   > SleeveLogic = RigVMFunctionReferenceNode()
   > RobeLogic_2 = RigVMFunctionReferenceNode()
   > Robe Wind = RigVMFunctionReferenceNode()
  ]}
]
hierarchy [
 root {loc=<0 0 0>, rot=<0 0 0 1>}
 pelvis {parent=root, loc=<0 -0.7071 93.1191>, rot=<0.0127 0.707 -0.0127 -0.707>}
 spine_01 {parent=pelvis, loc=<6.7494 -0.6388 0>, rot=<0 0 0.034 -0.9994>}
 spine_02 {parent=spine_01, loc=<7.2518 -0.035 0>, rot=<0 0 0.0238 -0.9997>}
 spine_03 {parent=spine_02, loc=<6.0104 0 0>, rot=<0 0 0.0149 -0.9999>}
 spine_04 {parent=spine_03, loc=<8.1024 0 0>, rot=<0 0 -0.0132 -0.9999>}
 spine_05 {parent=spine_04, loc=<10.0146 0 0>, rot=<0 0 -0.0015 -1>}
 neck_01 {parent=spine_05, loc=<11.3138 0.1529 0>, rot=<0 0 -0.2003 -0.9797>}
 neck_02 {parent=neck_01, loc=<4.5627 0 0>, rot=<0 0 0.0504 -0.9987>}
 head {parent=neck_02, loc=<5.3914 0 0>, rot=<0 0 0.115 -0.9934>}
 attach_cape {parent=spine_05, loc=<9.017 -11.714 0>, rot=<0.0616 -0.7044 -0.0616 -0.7044>}
 cape {parent=attach_cape, type=Null, loc=<0 1.0812 0>, rot=<0 0 0 1>}
 attach_backpack {parent=spine_05, loc=<-0.123 -13.473 0>, rot=<-0.0093 -0.707 0.0093 -0.707>}
 Backpack_BR {parent=attach_backpack, type=Null, loc=<0 2.0327 0>, rot=<0 0 0 1>}
 clavicle_l {parent=spine_05, loc=<4.3575 -0.1466 -2.3223>, rot=<0.0465 -0.7289 -0.009 -0.6829>}
 upperarm_l {parent=clavicle_l, loc=<15.5111 0 0>, rot=<0.0226 -0.3234 -0.0088 -0.9459>}
 lowerarm_l {parent=upperarm_l, loc=<26.6514 0 0>, rot=<0.0391 -0.0099 -0.2453 -0.9686>}
 hand_l {parent=lowerarm_l, loc=<22.3491 0 0>, rot=<-0.7081 -0.0507 0.0975 -0.6975>}
 thumb_01_l {parent=hand_l, loc=<1.3118 -0.3191 -2.459>, rot=<0.6546 -0.4042 0.0664 -0.6354>}
 thumb_02_l {parent=thumb_01_l, loc=<5.9336 -0.0448 0.0401>, rot=<-0.0037 0.0213 0.298 -0.9543>}
 thumb_03_l {parent=thumb_02_l, loc=<2.2223 0.3682 -0.0907>, rot=<0.0118 -0.013 0.0066 -0.9998>}
 index_metacarpal_l {parent=hand_l, loc=<1.6535 0.0183 -2.96>, rot=<0.0158 -0.0124 -0.0173 -0.9996>}
 index_01_l {parent=index_metacarpal_l, loc=<8.3008 0.0357 -0.0524>, rot=<-0.0044 0.0485 0.0924 -0.9945>}
 index_02_l {parent=index_01_l, loc=<3.3587 -0.3004 0.0452>, rot=<-0.002 0.0027 0.3225 -0.9466>}
 index_03_l {parent=index_02_l, loc=<2.5578 -0.123 -0.0252>, rot=<-0.002 0.0067 0.2558 -0.9667>}
 middle_metacarpal_l {parent=hand_l, loc=<2.1856 -0.1725 -0.5984>, rot=<0.006 0.0168 -0.0329 -0.9993>}
 middle_01_l {parent=middle_metacarpal_l, loc=<7.3275 -0.2447 0.0129>, rot=<-0.0208 0.0042 0.0858 -0.9961>}
 middle_02_l {parent=middle_01_l, loc=<4.0123 0.1401 0.1175>, rot=<-0.0218 0.0258 0.3512 -0.9357>}
 middle_03_l {parent=middle_02_l, loc=<2.6077 0.0534 0.1373>, rot=<0.0082 0.0188 0.1708 -0.9851>}
 ring_metacarpal_l {parent=hand_l, loc=<2.2074 -0.3625 1.3638>, rot=<-0.0284 0.0722 -0.0458 -0.9959>}
 ring_01_l {parent=ring_metacarpal_l, loc=<7.0295 -0.6197 -0.1446>, rot=<-0.0489 -0.0252 0.1892 -0.9804>}
 ring_02_l {parent=ring_01_l, loc=<3.6174 0.4647 -0.1027>, rot=<0 0 0.2806 -0.9598>}
 ring_03_l {parent=ring_02_l, loc=<2.4152 0.114 0.0199>, rot=<0 0 0.1189 -0.9929>}
 pinky_metacarpal_l {parent=hand_l, loc=<1.7616 -0.9753 3.3814>, rot=<-0.0397 0.0814 -0.069 -0.9935>}
 pinky_01_l {parent=pinky_metacarpal_l, loc=<7.1709 -0.4831 -0.1094>, rot=<-0.0654 -0.0035 0.2242 -0.9723>}
 pinky_02_l {parent=pinky_01_l, loc=<2.8274 0.0893 -0.1043>, rot=<0.0037 0.0011 0.2854 -0.9584>}
 pinky_03_l {parent=pinky_02_l, loc=<2.2046 0.0307 -0.0078>, rot=<-0.0002 0.0098 0.1445 -0.9895>}
 deform_wrist_dn_base_l {parent=hand_l, loc=<0 -2.2342 0>, rot=<0 0 0 -1>}
 deform_wrist_dn_l {parent=deform_wrist_dn_base_l, loc=<0 0 0>, rot=<0 0 0 1>}
 deform_wrist_up_base_l {parent=hand_l, loc=<0 2.2682 0>, rot=<0 0 0 -1>}
 deform_wrist_up_l {parent=deform_wrist_up_base_l, loc=<0 0 0>, rot=<0 0 0 1>}
 lowerarm_twist_02_l {parent=lowerarm_l, loc=<9.6199 0 0>, rot=<0 0 0 -1>}
 lowerarm_twist_01_l {parent=lowerarm_l, loc=<17.8549 0 0>, rot=<0 0 0 -1>}
 deform_elbow_in_base_l {parent=lowerarm_l, loc=<0.2112 5.259 0.8274>, rot=<0 0 0 -1>}
 deform_elbow_in_l {parent=deform_elbow_in_base_l, loc=<0 0 0>, rot=<0 0 0 1>}
 deform_elbow_out_base_l {parent=lowerarm_l, loc=<0 -4 0>, rot=<0 0 0 -1>}
 deform_elbow_out_l {parent=deform_elbow_out_base_l, loc=<0 0 0>, rot=<0 0 0 1>}
 upperarm_twist_01_l {parent=upperarm_l, loc=<0.0041 0 0>, rot=<0 0 0 -1>}
 deform_upperarm_bk_base_l {parent=upperarm_twist_01_l, loc=<1.0813 -7.2193 -1.9453>, rot=<0 0 0 1>}
 deform_upperarm_bk_l {parent=deform_upperarm_bk_base_l, loc=<0 0 0>, rot=<0 0 0 1>}
 deform_upperarm_fr_base_l {parent=upperarm_twist_01_l, loc=<-0.3691 7.0506 -2.765>, rot=<0 0 0 1>}
 deform_upperarm_fr_l {parent=deform_upperarm_fr_base_l, loc=<0 0 0>, rot=<0 0 0 1>}
 deform_upperarm_up_base_l {parent=upperarm_twist_01_l, loc=<-3.2216 -1.1391 3.3311>, rot=<0 0 0 1>}
 deform_upperarm_up_l {parent=deform_upperarm_up_base_l, loc=<0 0 0>, rot=<0 0 0 1>}
 upperarm_twist_02_l {parent=upperarm_l, loc=<13.7941 0 0>, rot=<0 0 0 -1>}
 dyn_sleeve_l_MSTR {parent=upperarm_l, loc=<11.4283 0 0>, rot=<0 0 0 1>}
 dyn_sleeve_l_tp_01 {parent=dyn_sleeve_l_MSTR, loc=<0 0 0>, rot=<0 0.606 0 -0.7955>}
 dyn_sleeve_l_tp_02 {parent=dyn_sleeve_l_tp_01, loc=<8.7558 0 0>, rot=<-0.0269 -0.5574 -0.0434 -0.8287>}
 dyn_sleeve_l_tp_03 {parent=dyn_sleeve_l_tp_02, loc=<7.5251 0 0>, rot=<-0.0038 0.0007 -0.0395 -0.9992>}
 dyn_sleeve_l_tp_04 {parent=dyn_sleeve_l_tp_03, loc=<9.4528 0 0>, rot=<0 0 -0.0257 -0.9997>}
 dyn_sleeve_l_tp_05 {parent=dyn_sleeve_l_tp_04, loc=<5.9265 0 0>, rot=<0 0 0 1>}
 dyn_sleeve_l_bt_01 {parent=dyn_sleeve_l_MSTR, loc=<0 0 0>, rot=<0 -0.5054 0 -0.8629>}
 dyn_sleeve_l_bt_02 {parent=dyn_sleeve_l_bt_01, loc=<10.2735 0 0>, rot=<0.025 0.4886 -0.0445 -0.871>}
 dyn_sleeve_l_bt_03 {parent=dyn_sleeve_l_bt_02, loc=<7.5251 0 0>, rot=<0.0052 -0.0346 -0.0393 -0.9986>}
 dyn_sleeve_l_bt_04 {parent=dyn_sleeve_l_bt_03, loc=<9.4528 0 0>, rot=<0.0006 -0.022 -0.0257 -0.9994>}
 dyn_sleeve_l_bt_05 {parent=dyn_sleeve_l_bt_04, loc=<5.9265 0 0>, rot=<0 0 0 1>}
 dyn_sleeve_l_bk_01 {parent=dyn_sleeve_l_MSTR, loc=<0 0 0>, rot=<0 0 0.4691 -0.8832>}
 dyn_sleeve_l_bk_02 {parent=dyn_sleeve_l_bk_01, loc=<7.96 0 0>, rot=<0 0 -0.5117 -0.8591>}
 dyn_sleeve_l_bk_03 {parent=dyn_sleeve_l_bk_02, loc=<7.5853 0 0>, rot=<0.0008 0 -0.0114 -0.9999>}
 dyn_sleeve_l_bk_04 {parent=dyn_sleeve_l_bk_03, loc=<9.2659 0 0>, rot=<0 0 -0.0257 -0.9997>}
 dyn_sleeve_l_bk_05 {parent=dyn_sleeve_l_bk_04, loc=<5.9265 0 0>, rot=<0 0 0 1>}
 dyn_sleeve_l_ft_01 {parent=dyn_sleeve_l_MSTR, loc=<0 0 0>, rot=<0 0 -0.5393 -0.8421>}
 dyn_sleeve_l_ft_02 {parent=dyn_sleeve_l_ft_01, loc=<7.96 0 0>, rot=<0 0 0.5063 -0.8623>}
 dyn_sleeve_l_ft_03 {parent=dyn_sleeve_l_ft_02, loc=<6.8854 0 0>, rot=<0.012 -0.0068 -0.2257 -0.9741>}
 dyn_sleeve_l_ft_04 {parent=dyn_sleeve_l_ft_03, loc=<8.2001 0 0>, rot=<0 0 0 1>}
 dyn_sleeve_l_ft_05 {parent=dyn_sleeve_l_ft_04, loc=<5.9265 0 0>, rot=<0 0 0 1>}
 clavicle_r {parent=spine_05, loc=<4.3575 -0.1466 2.3223>, rot=<0.7289 0.0465 -0.6829 0.009>}
 upperarm_r {parent=clavicle_r, loc=<-15.5111 0 0>, rot=<0.0226 -0.3234 -0.0088 -0.9459>}
 lowerarm_r {parent=upperarm_r, loc=<-26.6514 0 0>, rot=<0.0391 -0.0099 -0.2453 -0.9686>}
 hand_r {parent=lowerarm_r, loc=<-22.3491 0 0>, rot=<-0.7081 -0.0507 0.0975 -0.6975>}
 thumb_01_r {parent=hand_r, loc=<-1.3118 0.3191 2.459>, rot=<0.6546 -0.4042 0.0664 -0.6354>}
 thumb_02_r {parent=thumb_01_r, loc=<-5.9336 0.0448 -0.0401>, rot=<-0.0037 0.0213 0.298 -0.9543>}
 thumb_03_r {parent=thumb_02_r, loc=<-2.2223 -0.3682 0.0907>, rot=<0.0118 -0.013 0.0066 -0.9998>}
 index_metacarpal_r {parent=hand_r, loc=<-1.6535 -0.0183 2.96>, rot=<0.0158 -0.0124 -0.0173 -0.9996>}
 index_01_r {parent=index_metacarpal_r, loc=<-8.3008 -0.0357 0.0524>, rot=<-0.0044 0.0485 0.0924 -0.9945>}
 index_02_r {parent=index_01_r, loc=<-3.3587 0.3004 -0.0452>, rot=<-0.002 0.0027 0.3225 -0.9466>}
 index_03_r {parent=index_02_r, loc=<-2.5578 0.123 0.0252>, rot=<-0.002 0.0067 0.2558 -0.9667>}
 middle_metacarpal_r {parent=hand_r, loc=<-2.1856 0.1725 0.5984>, rot=<0.006 0.0168 -0.0329 -0.9993>}
 middle_01_r {parent=middle_metacarpal_r, loc=<-7.3275 0.2447 -0.0129>, rot=<-0.0208 0.0042 0.0858 -0.9961>}
 middle_02_r {parent=middle_01_r, loc=<-4.0123 -0.1401 -0.1175>, rot=<-0.0218 0.0258 0.3512 -0.9357>}
 middle_03_r {parent=middle_02_r, loc=<-2.6077 -0.0534 -0.1373>, rot=<0.0082 0.0188 0.1708 -0.9851>}
 ring_metacarpal_r {parent=hand_r, loc=<-2.2074 0.3625 -1.3638>, rot=<-0.0284 0.0722 -0.0458 -0.9959>}
 ring_01_r {parent=ring_metacarpal_r, loc=<-7.0295 0.6197 0.1446>, rot=<-0.0489 -0.0252 0.1892 -0.9804>}
 ring_02_r {parent=ring_01_r, loc=<-3.6174 -0.4647 0.1027>, rot=<0 0 0.2806 -0.9598>}
 ring_03_r {parent=ring_02_r, loc=<-2.4152 -0.114 -0.0199>, rot=<0 0 0.1189 -0.9929>}
 pinky_metacarpal_r {parent=hand_r, loc=<-1.7616 0.9753 -3.3814>, rot=<-0.0397 0.0814 -0.069 -0.9935>}
 pinky_01_r {parent=pinky_metacarpal_r, loc=<-7.1709 0.4831 0.1094>, rot=<-0.0654 -0.0035 0.2242 -0.9723>}
 pinky_02_r {parent=pinky_01_r, loc=<-2.8274 -0.0893 0.1043>, rot=<0.0037 0.0011 0.2854 -0.9584>}
 pinky_03_r {parent=pinky_02_r, loc=<-2.2046 -0.0307 0.0078>, rot=<-0.0002 0.0098 0.1445 -0.9895>}
 deform_wrist_dn_base_r {parent=hand_r, loc=<-0.0004 2.2345 -0.0001>, rot=<0 0 0 -1>}
 deform_wrist_dn_r {parent=deform_wrist_dn_base_r, loc=<0 0 0>, rot=<0 0 0 1>}
 deform_wrist_up_base_r {parent=hand_r, loc=<0.0003 -2.2684 0.0001>, rot=<0 0 0 -1>}
 deform_wrist_up_r {parent=deform_wrist_up_base_r, loc=<0 0 0>, rot=<0 0 0 1>}
 lowerarm_twist_02_r {parent=lowerarm_r, loc=<-9.6199 0 0>, rot=<0 0 0 -1>}
 lowerarm_twist_01_r {parent=lowerarm_r, loc=<-17.8549 0 0>, rot=<0 0 0 -1>}
 deform_elbow_in_base_r {parent=lowerarm_r, loc=<-0.2111 -5.2591 -0.8275>, rot=<0 0 0 -1>}
 deform_elbow_in_r {parent=deform_elbow_in_base_r, loc=<-0.0001 0 0.0002>, rot=<0 0 0 1>}
 deform_elbow_out_base_r {parent=lowerarm_r, loc=<0 4 -0.0001>, rot=<0 0 0 -1>}
 deform_elbow_out_r {parent=deform_elbow_out_base_r, loc=<0 0 0>, rot=<0 0 0 1>}
 upperarm_twist_01_r {parent=upperarm_r, loc=<-0.0041 0 0>, rot=<0 0 0 -1>}
 deform_upperarm_bk_base_r {parent=upperarm_twist_01_r, loc=<-1.0814 7.2192 1.9455>, rot=<0 0 0 1>}
 deform_upperarm_bk_r {parent=deform_upperarm_bk_base_r, loc=<0 0 0>, rot=<0 0 0 1>}
 deform_upperarm_fr_base_r {parent=upperarm_twist_01_r, loc=<0.3689 -7.0506 2.7652>, rot=<0 0 0 1>}
 deform_upperarm_fr_r {parent=deform_upperarm_fr_base_r, loc=<0 0 0>, rot=<0 0 0 1>}
 deform_upperarm_up_base_r {parent=upperarm_twist_01_r, loc=<3.2214 1.1391 -3.3308>, rot=<0 0 0 1>}
 deform_upperarm_up_r {parent=deform_upperarm_up_base_r, loc=<0 0 0>, rot=<0 0 0 -1>}
 upperarm_twist_02_r {parent=upperarm_r, loc=<-13.7941 0 0>, rot=<0 0 0 -1>}
 dyn_sleeve_r_MSTR {parent=upperarm_r, loc=<-11.4283 0 0>, rot=<0 -1 0 0>}
 dyn_sleeve_r_ft_01 {parent=dyn_sleeve_r_MSTR, loc=<0 0 0>, rot=<0 0 0.5393 -0.8421>}
 dyn_sleeve_r_ft_02 {parent=dyn_sleeve_r_ft_01, loc=<7.96 0 -0.0001>, rot=<0 0 -0.5063 -0.8623>}
 dyn_sleeve_r_ft_03 {parent=dyn_sleeve_r_ft_02, loc=<6.8857 0 -0.0003>, rot=<-0.012 -0.0068 0.2257 -0.9741>}
 dyn_sleeve_r_ft_04 {parent=dyn_sleeve_r_ft_03, loc=<8.1998 -0.0001 0.0003>, rot=<0 0 0 1>}
 dyn_sleeve_r_ft_05 {parent=dyn_sleeve_r_ft_04, loc=<5.9265 0 0.0001>, rot=<0 0 0 1>}
 dyn_sleeve_r_bk_01 {parent=dyn_sleeve_r_MSTR, loc=<0 0 0>, rot=<0 0 -0.4691 -0.8832>}
 dyn_sleeve_r_bk_02 {parent=dyn_sleeve_r_bk_01, loc=<7.9599 0.0002 0.0001>, rot=<0 0 0.5117 -0.8591>}
 dyn_sleeve_r_bk_03 {parent=dyn_sleeve_r_bk_02, loc=<7.5852 -0.0001 0.0002>, rot=<-0.0008 0 0.0114 -0.9999>}
 dyn_sleeve_r_bk_04 {parent=dyn_sleeve_r_bk_03, loc=<9.2659 0.0001 0>, rot=<0 0 0.0257 -0.9997>}
 dyn_sleeve_r_bk_05 {parent=dyn_sleeve_r_bk_04, loc=<5.9265 0 0>, rot=<0 0 0 1>}
 dyn_sleeve_r_bt_01 {parent=dyn_sleeve_r_MSTR, loc=<0 0 0>, rot=<0 -0.5054 0 -0.8629>}
 dyn_sleeve_r_bt_02 {parent=dyn_sleeve_r_bt_01, loc=<10.274 0 0.0001>, rot=<-0.025 0.4886 0.0445 -0.871>}
 dyn_sleeve_r_bt_03 {parent=dyn_sleeve_r_bt_02, loc=<7.5245 0 0.0006>, rot=<-0.0052 -0.0346 0.0393 -0.9986>}
 dyn_sleeve_r_bt_04 {parent=dyn_sleeve_r_bt_03, loc=<9.453 0 -0.0002>, rot=<-0.0006 -0.022 0.0257 -0.9994>}
 dyn_sleeve_r_bt_05 {parent=dyn_sleeve_r_bt_04, loc=<5.9263 0 0.0002>, rot=<0 0 0 1>}
 dyn_sleeve_r_tp_01 {parent=dyn_sleeve_r_MSTR, loc=<0 0 0>, rot=<0 0.606 0 -0.7955>}
 dyn_sleeve_r_tp_02 {parent=dyn_sleeve_r_tp_01, loc=<8.7555 0 -0.0003>, rot=<0.0269 -0.5574 0.0434 -0.8287>}
 dyn_sleeve_r_tp_03 {parent=dyn_sleeve_r_tp_02, loc=<7.5247 0 0.0007>, rot=<0.0038 0.0007 0.0395 -0.9992>}
 dyn_sleeve_r_tp_04 {parent=dyn_sleeve_r_tp_03, loc=<9.4532 0 -0.0005>, rot=<0 0 0.0257 -0.9997>}
 dyn_sleeve_r_tp_05 {parent=dyn_sleeve_r_tp_04, loc=<5.9265 0 0>, rot=<0 0 0 -1>}
 deform_pec_aim_base_l {parent=spine_04, loc=<7.8732 12.6388 -2.4123>, rot=<0.0102 -0.4602 0.3979 -0.7936>}
 deform_pec_aim_l {parent=deform_pec_aim_base_l, loc=<0 0 0>, rot=<0 0 0 1>}
 deform_pec_base_l {parent=deform_pec_aim_l, loc=<10.766 0 0>, rot=<0 0 0 1>}
 deform_pec_l {parent=deform_pec_base_l, loc=<0.055 0 0>, rot=<0 0 0 1>}
 deform_pec_aim_base_r {parent=spine_04, loc=<7.8727 12.6388 2.4123>, rot=<-0.4602 -0.0102 0.7936 0.3979>}
 deform_pec_aim_r {parent=deform_pec_aim_base_r, loc=<0 0 0>, rot=<0 0 0 1>}
 deform_pec_base_r {parent=deform_pec_aim_r, loc=<-10.7662 -0.0006 -0.0007>, rot=<0 0 0 1>}
 deform_pec_r {parent=deform_pec_base_r, loc=<0 0 0>, rot=<0 0 0 1>}
 thigh_l {parent=pelvis, loc=<-2.5693 -0.0109 -7.7501>, rot=<0.0012 0.0572 0.0155 -0.9982>}
 calf_l {parent=thigh_l, loc=<-39.0484 0 0>, rot=<0.0006 -0.0161 -0.0391 -0.9991>}
 foot_l {parent=calf_l, loc=<-42.561 0.1077 0.4292>, rot=<0.0418 -0.0455 0.0343 -0.9975>}
 ball_l {parent=foot_l, loc=<-6.1599 13.8264 0.6806>, rot=<0.0001 -0.0001 -0.7063 -0.7079>}
 calf_twist_02_l {parent=calf_l, loc=<-18.8064 -0.0967 0.0397>, rot=<0.0237 0.0095 -0.0035 -0.9997>}
 calf_twist_01_l {parent=calf_l, loc=<-35.813 -0.0967 0.0397>, rot=<0.0237 0.0095 -0.0035 -0.9997>}
 deform_knee_in_base_l {parent=calf_l, loc=<1.298 -5.2957 -0.0496>, rot=<0 0 0 -1>}
 deform_knee_in_l {parent=deform_knee_in_base_l, loc=<0 0 0>, rot=<0 0 0 1>}
 deform_knee_out_base_l {parent=calf_l, loc=<0 4 0>, rot=<0 0 0 -1>}
 deform_knee_out_l {parent=deform_knee_out_base_l, loc=<0 0 0>, rot=<0 0 0 1>}
 thigh_twist_01_l {parent=thigh_l, loc=<-21.6303 0 0>, rot=<0.0013 0 0 -1>}
 deform_groin_base_l {parent=thigh_l, loc=<-2.1908 8.1142 -1.7442>, rot=<0 0 0 -1>}
 deform_groin_l {parent=deform_groin_base_l, loc=<0 0 0>, rot=<0 0 0 1>}
 dyn_side_robe_l_MSTR {parent=thigh_l, loc=<-2.9389 0.6611 -6.6893>, rot=<0.0185 -0.72 -0.6937 0.0044>}
 dyn_side_robe_l_01 {parent=dyn_side_robe_l_MSTR, loc=<0 0 0>, rot=<0 0 0 1>}
 dyn_side_robe_l_02 {parent=dyn_side_robe_l_01, loc=<17.4495 0 0>, rot=<0 0 0 1>}
 dyn_side_robe_l_03 {parent=dyn_side_robe_l_02, loc=<19.6112 0 0>, rot=<0 0 -0.008 -1>}
 dyn_side_robe_l_04 {parent=dyn_side_robe_l_03, loc=<14.738 0 0>, rot=<0 0 0.0059 -1>}
 dyn_side_robe_l_05 {parent=dyn_side_robe_l_04, loc=<13.295 0 0>, rot=<0 0 0 1>}
 dyn_rear_robe_l_MSTR {parent=thigh_l, loc=<-1.9692 -10.5852 1.2626>, rot=<0.0062 -0.9984 -0.0553 -0.0046>}
 dyn_rear_robe_l_01 {parent=dyn_rear_robe_l_MSTR, loc=<0 0 0>, rot=<0 0 0 1>}
 dyn_rear_robe_l_02 {parent=dyn_rear_robe_l_01, loc=<17.4495 0 0>, rot=<0 0 0 1>}
 dyn_rear_robe_l_03 {parent=dyn_rear_robe_l_02, loc=<19.6112 0 0>, rot=<0 0 0.0021 -1>}
 dyn_rear_robe_l_04 {parent=dyn_rear_robe_l_03, loc=<14.738 0 0>, rot=<0 0 -0.0031 -1>}
 dyn_rear_robe_l_05 {parent=dyn_rear_robe_l_04, loc=<13.295 0 0>, rot=<0 0 0.0068 -1>}
 dyn_front_robe_l_MSTR {parent=thigh_l, loc=<-1.9529 10.7495 0.1103>, rot=<-0.0001 0.0568 0.9984 -0.0045>}
 dyn_front_robe_l_01 {parent=dyn_front_robe_l_MSTR, loc=<0 0 0>, rot=<0 0 0 1>}
 dyn_front_robe_l_02 {parent=dyn_front_robe_l_01, loc=<17.4495 0 0>, rot=<0 0 0 1>}
 dyn_front_robe_l_03 {parent=dyn_front_robe_l_02, loc=<19.6112 0 0>, rot=<0 0 -0.002 -1>}
 dyn_front_robe_l_04 {parent=dyn_front_robe_l_03, loc=<14.738 0 0>, rot=<0 0 0.0053 -1>}
 dyn_front_robe_l_05 {parent=dyn_front_robe_l_04, loc=<13.295 0 0>, rot=<0 0 -0.0091 -1>}
 deform_thigh_l {parent=thigh_l, loc=<0 0 0>, rot=<0 0 0 1>}
 deform_calf_l {parent=deform_thigh_l, loc=<-38.8556 0.0463 0.0008>, rot=<-0.0005 0.0191 0.3242 0.9458>}
 thigh_r {parent=pelvis, loc=<-2.5693 -0.0109 7.7501>, rot=<0.0572 -0.0012 0.9982 0.0155>}
 calf_r {parent=thigh_r, loc=<39.0484 0 0>, rot=<0.0006 -0.0161 -0.0391 -0.9991>}
 foot_r {parent=calf_r, loc=<42.561 -0.1077 -0.4292>, rot=<0.0418 -0.0455 0.0343 -0.9975>}
 ball_r {parent=foot_r, loc=<6.1599 -13.8264 -0.6806>, rot=<0.0001 -0.0001 -0.7063 -0.7079>}
 calf_twist_02_r {parent=calf_r, loc=<18.8064 0.0967 -0.0397>, rot=<0.0237 0.0095 -0.0035 -0.9997>}
 calf_twist_01_r {parent=calf_r, loc=<35.813 0.0967 -0.0397>, rot=<0.0237 0.0095 -0.0035 -0.9997>}
 deform_knee_in_base_r {parent=calf_r, loc=<-1.298 5.2957 0.0496>, rot=<0 0 0 -1>}
 deform_knee_in_r {parent=deform_knee_in_base_r, loc=<0 0 0>, rot=<0 0 0 1>}
 deform_knee_out_base_r {parent=calf_r, loc=<0 -4 0>, rot=<0 0 0 -1>}
 deform_knee_out_r {parent=deform_knee_out_base_r, loc=<0 0 0>, rot=<0 0 0 1>}
 thigh_twist_01_r {parent=thigh_r, loc=<21.6303 0 0>, rot=<0.0013 0 0 -1>}
 deform_groin_base_r {parent=thigh_r, loc=<2.1908 -8.1142 1.7442>, rot=<0 0 0 -1>}
 deform_groin_r {parent=deform_groin_base_r, loc=<0 0 0>, rot=<0 0 0 1>}
 dyn_side_robe_r_MSTR {parent=thigh_r, loc=<2.9389 -0.6611 6.6893>, rot=<0.0185 -0.72 -0.6937 0.0044>}
 dyn_side_robe_r_01 {parent=dyn_side_robe_r_MSTR, loc=<0 0 0>, rot=<0 0 0 1>}
 dyn_side_robe_r_02 {parent=dyn_side_robe_r_01, loc=<-17.4496 -0.0001 0>, rot=<0 0 0 1>}
 dyn_side_robe_r_03 {parent=dyn_side_robe_r_02, loc=<-19.6112 0 0>, rot=<0 0 -0.008 -1>}
 dyn_side_robe_r_04 {parent=dyn_side_robe_r_03, loc=<-14.7381 0 0>, rot=<0 0 0.0059 -1>}
 dyn_side_robe_r_05 {parent=dyn_side_robe_r_04, loc=<-13.295 0 0>, rot=<0 0 0 1>}
 dyn_rear_robe_r_MSTR {parent=thigh_r, loc=<1.9692 10.5852 -1.2626>, rot=<0.0062 -0.9984 -0.0553 -0.0046>}
 dyn_rear_robe_r_01 {parent=dyn_rear_robe_r_MSTR, loc=<0 0 0>, rot=<0 0 0 1>}
 dyn_rear_robe_r_02 {parent=dyn_rear_robe_r_01, loc=<-17.4495 -0.0001 0>, rot=<0 0 0 -1>}
 dyn_rear_robe_r_03 {parent=dyn_rear_robe_r_02, loc=<-19.6112 0 0>, rot=<0 0 0.0021 -1>}
 dyn_rear_robe_r_04 {parent=dyn_rear_robe_r_03, loc=<-14.738 0.0001 0>, rot=<0 0 -0.0031 -1>}
 dyn_rear_robe_r_05 {parent=dyn_rear_robe_r_04, loc=<-13.295 -0.0001 0>, rot=<0 0 0.0068 -1>}
 dyn_front_robe_r_MSTR {parent=thigh_r, loc=<1.9529 -10.7495 -0.1103>, rot=<-0.0001 0.0568 0.9984 -0.0045>}
 dyn_front_robe_r_01 {parent=dyn_front_robe_r_MSTR, loc=<0 0 0>, rot=<0 0 0 1>}
 dyn_front_robe_r_02 {parent=dyn_front_robe_r_01, loc=<-17.4495 0 0>, rot=<0 0 0 1>}
 dyn_front_robe_r_03 {parent=dyn_front_robe_r_02, loc=<-19.6112 0 0.0001>, rot=<0 0 -0.002 -1>}
 dyn_front_robe_r_04 {parent=dyn_front_robe_r_03, loc=<-14.738 0 -0.0001>, rot=<0 0 0.0053 -1>}
 dyn_front_robe_r_05 {parent=dyn_front_robe_r_04, loc=<-13.295 0 0>, rot=<0 0 -0.0091 -1>}
 deform_thigh_r {parent=thigh_r, loc=<0 0 0>, rot=<0 0 0 1>}
 deform_calf_r {parent=deform_thigh_r, loc=<38.8792 -0.0384 -0.0219>, rot=<0.0008 0.0216 0.3179 0.9479>}
 deform_glute_base_l {parent=pelvis, loc=<-8.8064 -8.8222 -8.4287>, rot=<0.0275 0.0501 -0.4472 -0.8926>}
 deform_glute_l {parent=deform_glute_base_l, loc=<0 0 0>, rot=<0 0 0 1>}
 deform_glute_base_r {parent=pelvis, loc=<-8.8062 -8.8222 8.4287>, rot=<0.0501 -0.0275 0.8926 -0.4472>}
 deform_glute_r {parent=deform_glute_base_r, loc=<0 0 0>, rot=<0 0 0 1>}
 dyn_front_robe_c_MSTR {parent=pelvis, loc=<-4.0208 12.5784 0>, rot=<0 0 0.9993 0.0366>}
 dyn_front_robe_c_01 {parent=dyn_front_robe_c_MSTR, loc=<0 0 0>, rot=<0 0 0 1>}
 dyn_front_robe_c_02 {parent=dyn_front_robe_c_01, loc=<17.4495 0 0>, rot=<0 0 0 1>}
 dyn_front_robe_c_03 {parent=dyn_front_robe_c_02, loc=<19.6112 0 0>, rot=<0 0 -0.0028 -1>}
 dyn_front_robe_c_04 {parent=dyn_front_robe_c_03, loc=<14.738 0 0>, rot=<0 0 -0.001 -1>}
 dyn_front_robe_c_05 {parent=dyn_front_robe_c_04, loc=<13.295 0 0>, rot=<0 0 -0.0091 -1>}
 dyn_rear_robe_c_MSTR {parent=pelvis, loc=<-4.913 -11.8434 0>, rot=<-0.005 -1 0 0>}
 dyn_rear_robe_c_01 {parent=dyn_rear_robe_c_MSTR, loc=<0 0 0>, rot=<0 0 0 1>}
 dyn_rear_robe_c_02 {parent=dyn_rear_robe_c_01, loc=<17.4495 0 0>, rot=<0 0 0 1>}
 dyn_rear_robe_c_03 {parent=dyn_rear_robe_c_02, loc=<19.6112 0 0>, rot=<0 0 0.0021 -1>}
 dyn_rear_robe_c_04 {parent=dyn_rear_robe_c_03, loc=<14.738 0 0>, rot=<0 0 0.0045 -1>}
 dyn_rear_robe_c_05 {parent=dyn_rear_robe_c_04, loc=<13.295 0 0>, rot=<0 0 0.0068 -1>}
 CR_root_no_rotation {parent=root, loc=<0 0 0>, rot=<0 0 0 1>}
]
variables [
 Calf_l_Bent {type=double, default=0}
 Calf_r_Bent {type=double, default=0}
 Is Not Emote Alpha {type=double, default=0}
 Is Skydiving Alpha {type=double, default=0}
 IsSkydiving {type=bool, default=False}
 L Foot Across Front {type=float, default=0}
 L Foot Across Rear {type=float, default=0}
 OutputParam {type=float, default=0}
 OutputParam_0 {type=float, default=0}
 PlayerSpeed {type=double, default=1}
 R Foot Across Front {type=float, default=0}
 R Foot Across Rear {type=float, default=0}
 Thigh_l_Inner {type=double, default=0}
 Thigh_l_Side {type=double, default=0}
 Thigh_l_backward {type=double, default=0}
 Thigh_l_forward {type=double, default=0}
 Thigh_l_forward_Standing {type=double, default=0}
 Thigh_r_Inner {type=double, default=0}
 Thigh_r_Side {type=double, default=0}
 Thigh_r_backward {type=double, default=0}
 Thigh_r_forward {type=double, default=0}
 Thigh_r_forward_Standing {type=double, default=0}
 pelvisLowToGround_INVERSE {type=double, default=0}
]
